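; Creative device Blueprint for a spline-based grind rail that players ride, broadcasting start and end grinding events.
; Device updates rebuild rail visuals from transient spline meshes and an async-loaded, style-specific variant component.
; Speed presets load curve tables into a per-actor scoped data registry, then user overrides adjust them.
; Enabled, visibility and minigame states drive collision and dismounts; edit mode shows hidden rails with a hologram material.
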(event EventBeginPlay
  (CallFunction|InitializeDataRegistryResolverScopes) ; Setup for per-actor scoped in the data registry
  (|CallParentFunction)
  (bind _class (Utilities|AsyncLoadClassAsset (Variables|Default|GetCameraModifierOverrideSoftClass))
    (:Completed
      (bind _ascamera_modifier_grinding (Utilities|Casting|CastToCameraModifier_GrindingClass _class))
      (Variables|Default|SetCameraModifierOverrideClass _ascamera_modifier_grinding))))

(event EventEndPlay
  (params (EndPlayReason :type EEndPlayReason%20Enum :default "Destroyed"))
  (|CallParentFunction EndPlayReason)
  (CallFunction|DismountGrindingPlayers))

(event Custom|UpdateDevice
  (Class|CreativeEnabledComponent|Initialize (Variables|Default|GetCreative_Enabled_Component) (Variables|Default|GetFortMinigameProgress) (Variables|UserOptions|Advanced|GetEnabled%20During%20Phase))
  (Class|FortCreativeSplineModificationComponent|OnSavedSplineDataLoaded (Variables|Default|GetFortCreativeSplineModification) (Variables|Internal|GetSplinePoints))
  (Default|UpdateRailMeshColors)
  (Default|UpdateVisibilityInGame)
  (CallFunction|LoadTuningsCurveTable (Utilities|Struct|BreakFGrindRailTuningPreset (Variables|UserOptions|Basic|GetSpeedPreset)))
  (CallFunction|CacheVisualStyleData)
  (CallFunction|SpawnGrindRailVariantComponent)
  (Networking|SwitchHasAuthority
    (:Authority
      (if (not (Editor|InEditor))
        (Class|FortCreativeSplineModificationComponent|GenerateControlPointsforSpline (Variables|Default|GetFortCreativeSplineModification))))
    (:Remote
      (Spline|UpdateTips))))

(event OnAnyPropertyChanged(ToyOptionsComponent)
  (CallFunction|UpdateDevice))

(event OnPropertiesLoaded(ToyOptionsComponent)
  (CallFunction|UpdateDevice))

(event EditorPreview|EventUpdateEditorPreview
  (params (bHasMoved :type Boolean))
  ;[0.399 0.7428 1] Spline Editing
  (Class|FortCreativeSplineModificationComponent|UpdateEditorPreview (Variables|Default|GetFortCreativeSplineModification) (Variables|Internal|GetSplinePoints)))

(event SplineProp|EventSaveSplinePointData
  ;[0.399 0.7428 1] Spline Editing
  ;;[0.4819 1 0.4834] SplinePropInterface Implementation
  (CallFunction|UpdatePropertyforLiveEdit (Utilities|String|ToString(Name) (Variables|/Script/FortniteGame.FortCreativeSplineModificationComponent|GetSplinePointsPropertyName (Variables|Default|GetFortCreativeSplineModification))))
  (CallFunction|Resave))

(event EditorPreview|EventSetupForEditorPreview
  ;[0.399 0.7428 1] Spline Editing
  (CallFunction|UpdateMaterialsforVisibility)
  (if (Utilities|Array|IsEmpty (Variables|Default|GetSplineStaticMesh))
    (CallFunction|CacheVisualStyleData)
    (CallFunction|SpawnGrindRailVariantComponent)))

(event SplineProp|EventSetSerializedSplinePoints
  (params (SplinePoints :type Array%20of%20Spline%20Point%20Serializable%20Structures))
  ;[0.399 0.7428 1] Spline Editing
  ;;[0.4819 1 0.4834] SplinePropInterface Implementation
  (Variables|Internal|SetSplinePoints SplinePoints))

(event SplineProp|EventOnSplinePointsChanged
  (params (NumSplinePointsPreChange :type Integer))
  ;[0.399 0.7428 1] Spline Editing
  ;;[0.4819 1 0.4834] SplinePropInterface Implementation
  (Spline|UpdateVisuals NumSplinePointsPreChange)
  (Default|UpdateVisibilityInGame) ; Re-initialize the visible component with the new set of spline meshes
)

(event EventPostLevelSaveSpawnCleanup
  ;[0.399 0.7428 1] Spline Editing
  ;;[1 0.8866 0.4618] Since generated components are spawned as transient, the client/server is unable to resolve the references in the root component's AttachedChildren when the device is moved/rotated in UEFN, which leads to the AttachedChildren array being stomped with null elements and prevents the transform changes from propagating. So manually fix up the transforms here after the data has been serialized
  (|CallParentFunction)
  (Class|FortCreativeSplineModificationComponent|UpdateTransientComponentTransforms (Variables|Default|GetFortCreativeSplineModification)))

(event SplineProp|EventUpdateSplineMeshAtIndex
  (params (SplineMeshIndex :type Integer))
  ;[0.399 0.7428 1] Spline Editing
  ;;[0.4819 1 0.4834] SplinePropInterface Implementation
  (Spline|UpdateSpineMeshatIndex SplineMeshIndex))

(event EventOnPlayerBeganGrinding
  (params (GrindingPawn :type Fort%20Player%20Pawn%20Object%20Reference))
  (bind _returnvalue (Actor|GetComponentbyClass GrindingPawn "/GrindRail/Blueprints/BP_FortPawnGrindingProvider.BP_FortPawnGrindingProvider_C"))
  ;[1 0.5735 0.3708] Gameplay Events/Functions
  (Utilities|IsValid _returnvalue
    (:Is%20Valid
      (Utilities|Array|Add (Variables|Default|GetAttachedGrindingProviders) _returnvalue)
      (GameplayEvent|BroadcastEvent (Variables|UserOptions|Events|GetOn%20Started%20Grinding) 0 (Utilities|Struct|MakeGameplayEventGrindRailPlayerStartedGrinding (Pawn|GetController GrindingPawn) self)))))

(event EventOnPlayerEndedGrinding
  (params (Pawn :type Fort%20Player%20Pawn%20Object%20Reference))
  (bind _returnvalue (Actor|GetComponentbyClass Pawn "/GrindRail/Blueprints/BP_FortPawnGrindingProvider.BP_FortPawnGrindingProvider_C"))
  ;[1 0.5735 0.3708] Gameplay Events/Functions
  (Utilities|IsValid _returnvalue
    (:Is%20Valid
      (Utilities|Array|RemoveItem (Variables|Default|GetAttachedGrindingProviders) _returnvalue)
      (GameplayEvent|BroadcastEvent (Variables|UserOptions|Events|GetOn%20Ended%20Grinding) 0 (Utilities|Struct|MakeGameplayEventGrindRailPlayerEndedGrinding (Pawn|GetController Pawn) self)))))

(event Custom|EnableEvent
  ;[1 0.5735 0.3708] Gameplay Events/Functions
  (Class|CreativeEnabledComponent|SetEnabled (Variables|Default|GetCreative_Enabled_Component) true))

(event Custom|ShowTrackEvent
  ;[1 0.5735 0.3708] Gameplay Events/Functions
  (CallFunction|SetComponentsVisibility true))

(event Custom|DisableEvent
  ;[1 0.5735 0.3708] Gameplay Events/Functions
  (Class|CreativeEnabledComponent|SetEnabled (Variables|Default|GetCreative_Enabled_Component) false))

(event Custom|HideTrackEvent
  ;[1 0.5735 0.3708] Gameplay Events/Functions
  (CallFunction|SetComponentsVisibility false))

(event Custom|SetComponentsVisibility
  (params (bVisible :type Boolean))
  ;[1 0.5735 0.3708] Gameplay Events/Functions
  (for _array_element (SplineProp|GetTransientSceneComponents)
    (Rendering|SetVisibility _array_element bVisible false)
    (Development|SetHiddeninGame _array_element (not bVisible) false))
  (CallFunction|UpdateCollision))

(event MinigameStateChangedDelegate(FortMinigameProgress)
  (params (Minigame :type Fort%20Minigame%20Object%20Reference) (NewMinigameState :type EFortMinigameState%20Enum :default "PreGame"))
  ;[1 0.939 0.5021] Gameplay State Changes
  ;;[0.4745 1 0.8795] Stop grinding for players when the scoreboardstops showing after the game either in between rounds, starting a new match, or returning to edit mode
  (if (Math|Boolean|ORBoolean (Utilities|Enum|Equal(Enum) NewMinigameState "PostGameReset") (Utilities|Enum|Equal(Enum) NewMinigameState "Transitioning") (Utilities|Enum|Equal(Enum) NewMinigameState "PostGameAbandon"))
    (Class|CreativeEnabledComponent|SetEnabled (Variables|Default|GetCreative_Enabled_Component) false)
    (CallFunction|DismountGrindingPlayers)))

(event MinigameEndedDelegate(FortMinigameProgress)
  ;[1 0.939 0.5021] Gameplay State Changes
  (Verse|ResetVerseOverrides))

(event OnEnabledStateChanged(Creative_Enabled_Component)
  (params (Enabled :type Boolean))
  ;[1 0.939 0.5021] Gameplay State Changes
  (CallFunction|UpdateCollision)
  (if (and (not Enabled) (Variables|UserOptions|Advanced|GetbEndGrindingOnDisable))
    (CallFunction|DismountGrindingPlayers)))

(event OnVisibilityChanged(Creative_VisibleInGame_Component)
  (params (Visible :type Boolean))
  ;[1 0.939 0.5021] Gameplay State Changes
  (CallFunction|UpdateMaterialsforVisibility))

(event OnAddedtoMinigame(FortMinigameProgress)
  (params (Minigame :type Fort%20Minigame%20Object%20Reference))
  ;[1 0.939 0.5021] Gameplay State Changes
  (Utilities|IsValid Minigame
    (:Is%20Valid
      (Minigame|BindEventtoOnPlayModeChangedDelegate Minigame (delegate Custom|OnPlayModeChanged)))))

(event Custom|OnPlayModeChanged
  (params (Minigame :type Fort%20Minigame%20Object%20Reference) (bIsInPlayMode :type Boolean))
  ;[1 0.939 0.5021] Gameplay State Changes
  (CallFunction|UpdateCollision))

(event Custom|LoadTuningsCurveTable
  (params (AssetToLoad :type Soft%20Object%20Reference))
  ;[0.6278 0.176 1] Async Loading
  (if (Utilities|IsValidSoftObjectReference AssetToLoad)
    (bind _object (Utilities|AsyncLoadAsset AssetToLoad)
      (:Completed
        (ScalableFloat|AddCurveTabletoDataRegistrywithContext _object (Variables|DataRegistry|GetDataRegistryName))
        (Default|UpdateSpeedSettings))) ; Only used during gameplay anyway, so no need for blocking
  ))

(event Custom|LoadAndSpawnVariantComponent
  (params (VariantSoftClass :type Grind%20Rail%20Variant%20Component%20Base%20Soft%20Class%20Reference))
  ;[0.6278 0.176 1] Async Loading
  (bind _output_get (Variables|VisualData|SetLoadingVariantClass VariantSoftClass)) ; Cache the currently loading class to avoid spawning 2 components in one frame
  (bind _loaded_class (Creative|Loading|AsyncLoadClassEditorSafe _output_get)
    (:Loaded
      (Variables|VisualData|SetLoadingVariantClass 0)
      (bind _asactor_component (Utilities|Casting|CastToActorComponentClass _loaded_class))
      (bind _spawnedcomponent (Fortnite|Utilities|SpawnTransientComponent _asactor_component self))
      (bind _asgrind_rail_variant_component_base (Utilities|Casting|CastToGrindRailVariantComponent_Base _spawnedcomponent))
      (bind _output_get_2 (Variables|Default|SetGrindRailVariantComponent _asgrind_rail_variant_component_base))
      (Class|GrindRailVariantComponentBase|InitializeVariantComponent _output_get_2 (Variables|VisualData|GetVisualStyleData))
      (CallFunction|InitializeSceneComponentsforVisibility))))

(event EventOnActorMovedToNewLocation
  ;[1 1 1] If device moved during edit mode, update spline
  (if (not (Minigame|IsMinigameinProgress (Variables|Default|GetFortMinigameProgress)))
    (CallFunction|UpdateDevice)))

(fn AddInteractionRailMeshComponent
  ;[1 1 1] AddSplineMeshComponent node spawns the component as Movable so that the static mesh can be set outside of this being called during construction
  (if (Utilities|Enum|Equal(Enum) (Variables|UserOptions|Basic|GetVisual%20Style) "NewEnumerator2")
    (bind _spawnedcomponent (Fortnite|Utilities|SpawnTransientSceneComponent "/CRD_GrindRail/CRD_GrindRailInteractionSplineMeshComponent.CRD_GrindRailInteractionSplineMeshComponent_C" self)) ; Transient to prevent sending the component across UEFN LiveEdit
    (Utilities|IsValid _spawnedcomponent
      (:Is%20Valid
        (Utilities|Array|Add (Variables|Default|GetInteractionMeshes) _spawnedcomponent)
        (Components|StaticMesh|SetStaticMesh _spawnedcomponent (Variables|Default|GetMesh_Interaction))
        (Transformation|SetMobility _spawnedcomponent "Static"))
      (:Is%20Not%20Valid
        _))
    (return _spawnedcomponent)
    (else
      (return 0))))

(fn AddSolidRailMeshComponent
  (params (SegmentLength :type Float%20%28double-precision%29) (FirstPoint :type Integer) (SecondPoint :type Integer))
  ;[1 1 1] AddSplineMeshComponent node spawns the component as Movable so that the static mesh can be set outside of this being called during construction
  (bind _spawnedcomponent (Fortnite|Utilities|SpawnTransientSceneComponent "/CRD_GrindRail/CRD_GrindRailSplineMeshComponent.CRD_GrindRailSplineMeshComponent_C" self (Variables|Default|GetDefaultSceneRoot))) ; Transient to prevent sending the component across UEFN LiveEdit
  (Utilities|IsValid _spawnedcomponent
    (:Is%20Valid
      (Components|StaticMesh|SetStaticMesh _spawnedcomponent (Default|SelectSolidMeshforSegment (Variables|GetFirstPoint) (Variables|GetSecondPoint)))
      (Transformation|SetMobility _spawnedcomponent "Static"))
    (:Is%20Not%20Valid
      _))
  (return _spawnedcomponent))

(fn AllowInverseScaling
  (return false))

(fn AllowSprinting
  (return (Variables|UserOptions|Basic|GetbAllowSprinting)))

(fn BlueprintCanInteract
  (params (InteractingPawn :type Fort%20Pawn%20Object%20Reference) (InteractionBeingAttempted :type EInteractionBeingAttempted%20Enum :default "FirstInteraction") (InteractionType :type TInteractionType%20Enum :default "IT_NoInteraction"))
  (return (or (CallFunction|IsGrindRailEnabled) (not (Minigame|IsinPlayMode self :!class /Script/FortniteGame.MinigameBlueprintLibrary)))))

(fn BlueprintGetInteractionString
  (params (InteractingPawn :type Fort%20Pawn%20Object%20Reference) (InteractionBeingAttempted :type EInteractionBeingAttempted%20Enum :default "FirstInteraction"))
  (if (Class|CreativeCommonDeviceFunctionLibrary|IsinPlayMode self)
    (bind _returnvalue (|CallParentFunction InteractingPawn InteractionBeingAttempted))
    (return _returnvalue)
    (else
      (return "NSLOCTEXT(\"[D6BEE741841A788A65D7830521832910]\", \"3EAE79B64B43B767754D009DF65D5248\", \"CUSTOMIZE GRIND RAIL\")"))))

(fn BlueprintOnLocalInteract
  (params (InteractingPawn :type Fort%20Player%20Pawn%20Object%20Reference) (InteractionBeingAttempted :type EInteractionBeingAttempted%20Enum :default "FirstInteraction"))
  (if (Class|CreativeCommonDeviceFunctionLibrary|IsinPlayMode self)
    (bind _returnvalue (|CallParentFunction InteractingPawn InteractionBeingAttempted))
    (return _returnvalue)
    (else
      (return false))))

(fn BuildDataRegistryResolverScope
  (return (Variables|DataRegistry|GetResolver%20Scopes) 1 true))

(fn CacheVisualStyleData
  (bind _returnvalue (Utilities|GetDataTableRow (Variables|VisualData|GetStyleDataTable) (Utilities|String|StringToName (Utilities|String|EnumtoString (Variables|UserOptions|Basic|GetVisual%20Style)))))
  (bind _output_get (Variables|VisualData|SetVisualStyleData (Utilities|Struct|BreakFGrindRailStyleDataRegistryRow _returnvalue)))
  (return _output_get))

(fn DestroySplineMeshAtIndex
  (params (MeshIndex :type Integer))
  (if (Utilities|Array|IsValidIndex (Variables|Default|GetSplineStaticMesh) MeshIndex)
    (Components|DestroyComponent (Utilities|Array|Get(acopy) (Variables|Default|GetSplineStaticMesh) (Variables|GetMeshIndex)))
    (else
      (Development|PrintString (Utilities|String|Append "[Grind Rail] Invalid Spline Mesh index: " (Utilities|String|ToString(Integer) MeshIndex)))))
  (if (Utilities|Array|IsValidIndex (Variables|Default|GetInteractionMeshes) (Variables|GetMeshIndex))
    (Components|DestroyComponent (Utilities|Array|Get(acopy) (Variables|Default|GetInteractionMeshes) (Variables|GetMeshIndex)))))

(fn DismountGrindingPlayers
  (bind _attachedgrindingproviders (Variables|Default|GetAttachedGrindingProviders))
  (bind _output_get (LocalVariables|SetTempAttachedGrindingProvider _attachedgrindingproviders))
  (for _array_element _output_get
    (Class|FortPawnComponentGrindingProvider|BPonDismount _array_element))
  (Utilities|Array|Clear _attachedgrindingproviders))

(fn ForceGridAlignment
  (return false))

(fn GetEmissiveMaterials
  (if (CallFunction|IsLoadingVariantComponent)
    (return (Utilities|Array|MakeArray (Variables|Material|GetMaterialOfEmisssive)))
    (else
      (return (Class|GrindRailVariantComponentBase|GetSplineMaterials (Variables|Default|GetGrindRailVariantComponent) :!class /CRD_GrindRail/Blueprints/Components/GrindRailVariantComponent_Base.GrindRailVariantComponent_Base_C)))))

(fn GetAdditionalSelectableActors
  (return 0))

(fn GetBlockPitchAndRoll
  (return false))

(fn GetCacheContextOverride
  (return 0))

(fn GetCameraModifierOverrideClass
  (return (Variables|Default|GetCameraModifierOverrideClass) (Utilities|IsValidClass (Variables|Default|GetCameraModifierOverrideClass))))

(fn GetCollisionStaticMesh
  (return "0, 0, 0" 0))

(fn GetCreativeActorBounds
  (return false "0, 0, 0"))

(fn GetCreativeActorOrigin
  (bind (_origin _boxextent) (Collision|GetComponentBounds (Variables|Default|GetSpline)))
  ;[1 0.586 0 0.2] Using Spline Bounds for Actor Origin
  (return true (- (+ _origin _boxextent) (Transformation|GetActorLocation))))

(fn GetCreativeActorRelativeTransform
  (return false 0))

(fn GetGridAxisGuideAttachmentPoint
  (bind (_inputkey _position) (Utilities|Struct|BreakSplinePoint (Spline|GetSplinePointAt (Variables|Default|GetSpline) (- (Spline|GetNumberOfSplinePoints (Variables|Default|GetSpline)) 1) "Local")))
  (bind (_inputkey_2 _position_2) (Utilities|Struct|BreakSplinePoint (Spline|GetSplinePointAt (Variables|Default|GetSpline) 0 "Local")))
  (return true true (/ (+ _position_2 _position) 2)))

(fn GetMeshComponents
  (bind _starttip (Variables|Default|GetStartTip))
  (bind _endtip (Variables|Default|GetEndTip))
  (for _array_element (Variables|Default|GetSplineStaticMesh)
    (Utilities|Array|Add (LocalVariables|GetVisibleMeshes) _array_element))
  ;[1 0.9776 0.4567] Add rail tip meshes
  (Utilities|IsValid _starttip
    (:Is%20Valid
      (Utilities|Array|Add (LocalVariables|GetVisibleMeshes) _starttip))
    (:Is%20Not%20Valid
      _))
  (Utilities|IsValid _endtip
    (:Is%20Valid
      (Utilities|Array|Add (LocalVariables|GetVisibleMeshes) _endtip))
    (:Is%20Not%20Valid
      _))
  (return (LocalVariables|GetVisibleMeshes)))

(fn GetRailColorAndEmissive
  (return (Math|Color|SelectColor (Variables|Verse|GetRailColorOverride) (Variables|UserOptions|Basic|GetRail%20Color) (Variables|Verse|GetbIsRailColorOverridden)) (Variables|Material|GetColorEmissiveScale)))

(fn GetRotationForTip
  (params (bEndTip :type Boolean))
  (bind (_returnvalue_location _returnvalue_rotation) (Math|Transform|BreakTransform (Spline|GetTransformatDistanceAlongSpline (Variables|Default|GetSpline) (Math|Float|SelectFloat (- (Spline|GetSplineLength (Variables|Default|GetSpline)) (Utilities|Array|Get(acopy) (Variables|VisualData|GetSplineSampleOffsetForTipRotation) (Math|Conversions|ToInteger(Byte) (Variables|UserOptions|Basic|GetVisual%20Style)))) (Utilities|Array|Get(acopy) (Variables|VisualData|GetSplineSampleOffsetForTipRotation) (Math|Conversions|ToInteger(Byte) (Variables|UserOptions|Basic|GetVisual%20Style))) bEndTip) "World" false)))
  (return (Math|Rotator|SelectRotator (Math|Conversions|RotationFromXVector (Math|Vector|NegateVector (Math|Vector|GetForwardVector _returnvalue_rotation))) _returnvalue_rotation (Variables|GetbEndTip))))

(fn GetSavedSplinePoints
  (return (Variables|Internal|GetSplinePoints)))

(fn GetScaleMax
  (params (DefaultScaleMax :type Float%20%28single-precision%29))
  (return 1.0))

(fn GetScaleMin
  (params (DefaultScaleMin :type Float%20%28single-precision%29))
  (return 1.0))

(fn GetShouldUseHiddenMaterial
  (return (Default|ShouldUseHiddenMaterial)))

(fn GetSpawnedActors
  (return 0))

(fn GetSplineComponent
  (return (Variables|Default|GetSpline)))

(fn GetTransientSceneComponents
  (bind _output_get (LocalVariables|SetComponents (Variables|Default|GetSplineStaticMesh)))
  (Utilities|Array|Add _output_get (Variables|Default|GetStartTip))
  (Utilities|Array|Add _output_get (Variables|Default|GetEndTip))
  (Utilities|IsValid (Variables|Default|GetGrindRailVariantComponent)
    (:Is%20Valid
      (Utilities|Array|AppendArray (LocalVariables|GetComponents) (Class|GrindRailVariantComponentBase|GetSceneComponents (Variables|Default|GetGrindRailVariantComponent))))
    (:Is%20Not%20Valid
      _))
  (return (LocalVariables|GetComponents)))

(fn HasAdditionalSelectableActors
  (return false))

(fn HasAdditionalValidPlacementLogic
  (return true))

(fn HasValidRotation
  (return (and (< (Math|Float|Absolute(Float) (.roll (.rotation (Transformation|GetActorTransform)))) (Variables|Restrictions|GetRollMaxThreshold)) (< (Math|Float|Absolute(Float) (.pitch (.rotation (Transformation|GetActorTransform)))) (Variables|Restrictions|GetPitchMaxThreshold)))))

(fn IgnoreWorldGridScaleOverride
  (return false))

(fn InitializeDataRegistryResolverScopes
  (Networking|SwitchHasAuthority
    (:Authority
      (if (Utilities|Array|IsEmpty (Variables|DataRegistry|GetResolver%20Scopes))
        (Variables|DataRegistry|SetResolver%20Scopes (Utilities|Array|MakeArray (SpatialGameplay|GetSpatialGameplayInterfaceFNameforActor self) (Utilities|String|StringToName (Utilities|GetDisplayName self))))
        (return)))))

(fn InitializeSceneComponentsForVisibility
  (Initialization|InitializeWithSceneComponents (Variables|Default|GetCreative_VisibleInGame_Component) (SplineProp|GetTransientSceneComponents :!class /Script/FortniteGame.SplinePropInterface) (Utilities|Map|MakeMap) (Variables|UserOptions|Advanced|GetVisibleDuringPhase)))

(fn IsGrindRailEnabled
  (return (Class|CreativeEnabledComponent|IsEnabled (Variables|Default|GetCreative_Enabled_Component))))

(fn IsLoadingVariantComponent
  (return (or (not (Utilities|IsValid (Variables|Default|GetGrindRailVariantComponent) :!class /Script/Engine.KismetSystemLibrary)) (Utilities|IsValidSoftClassReference (Variables|VisualData|GetLoadingVariantClass)))))

(fn IsStartingGrindingFromWalkingEnabled
  (return (Variables|UserOptions|Advanced|GetStart%20Grinding%20From%20Walking)))

(fn IsValidPlacement
  (return (Default|HasValidRotation)))

(fn IsVisible
  (if (Utilities|Array|IsValidIndex (Variables|Default|GetSplineStaticMesh))
    (return (Rendering|IsVisible (Utilities|Array|Get(acopy) (Variables|Default|GetSplineStaticMesh))))
    (else
      (return false))))

(fn MigrateLegacyUserOptions
  (bind _gravityforcewhengoingdown (Class|GetClassDefaults "/CRD_GrindRail/Device_GrindRail.Device_GrindRail_C"))
  (bind _shootingspeedmultiplier (Class|GetClassDefaults "/CRD_GrindRail/Device_GrindRail.Device_GrindRail_C"))
  (bind _sprintacceleration (Class|GetClassDefaults "/CRD_GrindRail/Device_GrindRail.Device_GrindRail_C"))
  (bind _leanacceleration (Class|GetClassDefaults "/CRD_GrindRail/Device_GrindRail.Device_GrindRail_C"))
  (bind _gravityforcewhengoingup (Class|GetClassDefaults "/CRD_GrindRail/Device_GrindRail.Device_GrindRail_C"))
  (|CallParentFunction)
  (Utilities|FlowControl|Sequence
    (:then_0
      (if (not (Math|Float|NearlyEqual(Float) (Variables|Legacy|GetShootingSpeedMultiplier) _shootingspeedmultiplier))
        (Variables|UserOptions|Overrides|SetShootingSpeedMult_Override true)
        (Variables|UserOptions|Basic|SetShootingSpeedMult (Variables|Legacy|GetShootingSpeedMultiplier))
        (Variables|Legacy|SetShootingSpeedMultiplier _shootingspeedmultiplier)
        (LocalVariables|SetbOptionsMigrated true)))
    (:then_1
      (if (not (Math|Float|NearlyEqual(Float) (Variables|Legacy|GetSprintAcceleration) _sprintacceleration))
        (Variables|UserOptions|Overrides|SetAcceleration_Sprint_Override true)
        (Variables|UserOptions|Advanced|SetAcceleration_Sprint (Variables|Legacy|GetSprintAcceleration))
        (Variables|Legacy|SetSprintAcceleration _sprintacceleration)
        (LocalVariables|SetbOptionsMigrated true)))
    (:then_2
      (if (not (Math|Float|NearlyEqual(Float) (Variables|Legacy|GetLeanAcceleration) _leanacceleration))
        (Variables|UserOptions|Overrides|SetAcceleration_Lean_Override true)
        (Variables|UserOptions|Advanced|SetAcceleration_Lean (Variables|Legacy|GetLeanAcceleration))
        (Variables|Legacy|SetLeanAcceleration _leanacceleration)
        (LocalVariables|SetbOptionsMigrated true)))
    (:then_3
      (if (not (Math|Float|NearlyEqual(Float) (Variables|Legacy|GetGravityForceWhenGoingUp) _gravityforcewhengoingup))
        (Variables|UserOptions|Overrides|SetGravityForce_GoingUp_Override true)
        (Variables|UserOptions|Advanced|SetGravityForce_GoingUp (Variables|Legacy|GetGravityForceWhenGoingUp))
        (Variables|Legacy|SetGravityForceWhenGoingUp _gravityforcewhengoingup)
        (LocalVariables|SetbOptionsMigrated true)))
    (:then_4
      (if (not (Math|Float|NearlyEqual(Float) (Variables|Legacy|GetGravityForceWhenGoingDown) _gravityforcewhengoingdown))
        (Variables|UserOptions|Overrides|SetGravityForce_GoingDown_Override true)
        (Variables|UserOptions|Advanced|SetGravityForce_GoingDown (Variables|Legacy|GetGravityForceWhenGoingDown))
        (Variables|Legacy|SetGravityForceWhenGoingDown _gravityforcewhengoingdown)
        (LocalVariables|SetbOptionsMigrated true)))
    (:then_5
      (if (LocalVariables|GetbOptionsMigrated)
        (OptionData|RefreshPropertyOverrides (Variables|Default|GetToyOptionsComponent))))
    (:then_6
      (if (< (Variables|UserOptions|Advanced|GetMinimumSpeed) 1.0)
        (Variables|UserOptions|Advanced|SetMinimumSpeed 1.0) ; Restrict MinSpeed min value to 1 as the underlying logic doesn't work well with a speed of 0, and 1 is so small it's basically not moving
        (LocalVariables|SetbOptionsMigrated true)))))

(fn OnPreviewEnd
  (return false))

(fn OnPreviewStart
  (return false))

(fn OnRep_RailColorOverride
  (bind (_railcoloroverride_r _railcoloroverride_g _railcoloroverride_b _railcoloroverride_a) (Math|Color|BreakColor (Variables|GetRailColorOverride_R)))
  (Variables|Verse|SetbIsRailColorOverridden (!= _railcoloroverride_a 0))
  (Default|UpdateRailMeshColors))

(fn ResetVerseOverrides
  (if (Variables|Verse|GetbIsRailColorOverridden)
    (Variables|Verse|SetRailColorOverride (LocalVariables|GetColorToTriggerReset))))

(fn SelectSolidMeshForSegment
  (params (FirstSplinePoint :type Integer) (SecondSplinePoint :type Integer))
  (if (CallFunction|IsLoadingVariantComponent)
    (return (Variables|Default|GetMesh_Rail_Default))
    (else
      (return (Class|GrindRailVariantComponentBase|GetSplineMesh (Variables|Default|GetGrindRailVariantComponent) (CallFunction|ShouldUseSegmentedCollisionMesh FirstSplinePoint SecondSplinePoint) (Math|Float|Absolute(Float) (- (Spline|GetDistanceAlongSplineatSplinePoint (Variables|Default|GetSpline) (Variables|GetSecondSplinePoint)) (Spline|GetDistanceAlongSplineatSplinePoint (Variables|Default|GetSpline) (Variables|GetFirstSplinePoint)))))))))

(fn SetRailColor
  (params (NewColor_R :type Float%20%28single-precision%29) (NewColor_G :type Float%20%28single-precision%29) (NewColor_B :type Float%20%28single-precision%29) (NewColor_A :type Float%20%28single-precision%29))
  (Variables|SetRailColorOverride_R (Math|Color|MakeColor NewColor_R NewColor_G NewColor_B)) ; Force an alpha of 1
)

(fn SetSplineMeshMaterials
  (params (SplineMeshComp :type Spline%20Mesh%20Component%20Object%20Reference))
  (for (_array_element _array_index) (Default|GetEmissiveMaterials)
    (Rendering|Material|SetMaterial SplineMeshComp _array_index (select (Default|ShouldUseHiddenMaterial) (Variables|VisualData|GetInvisiblePreviewMaterial) _array_element))))

(fn ShouldApplyFallDamageImmunity
  (return (Variables|UserOptions|Advanced|GetbApplyFallDamageImmunity)))

(fn ShouldBlockPlacement
  (params (MoveTool :type Fort%20Creative%20Move%20Tool%20Object%20Reference))
  (return (not (Default|HasValidRotation))))

(fn ShouldSoftCapMaxSpeeds
  (return (Utilities|Select :Index (Variables|UserOptions|Advanced|GetMaxSpeedType) :NewEnumerator0 true :NewEnumerator1 false)))

(fn ShouldUseSegmentedCollisionMesh
  (params (FirstSplinePointIndex :type Integer) (SecondSplinePointIndex :type Integer))
  (if (not (Editor|InEditor))
    (if (>= (Spline|GetNumberOfSplinePoints (Variables|Default|GetSpline)) (Variables|GetSecondSplinePointIndex))
      (return (or (and (< (Math|Vector|DotProduct (Math|Vector|Normalize (- (Spline|GetLocationatSplinePoint (Variables|Default|GetSpline) (Variables|GetSecondSplinePointIndex) "World") (Spline|GetLocationatSplinePoint (Variables|Default|GetSpline) (Variables|GetFirstSplinePointIndex) "World"))) (Math|Vector|Normalize (Spline|GetTangentatSplinePoint (Variables|Default|GetSpline) (Variables|GetFirstSplinePointIndex) "World"))) (ScalableFloat|GetValueatLevel (Variables|HF|GetCP_PosDirThreshold))) (or (> (Math|Vector|VectorLengthSquared (Spline|GetTangentatSplinePoint (Variables|Default|GetSpline) (Variables|GetFirstSplinePointIndex) "World")) (Math|Float|Square (ScalableFloat|GetValueatLevel (Variables|HF|GetCP_TangentThreshold)))) (> (Math|Vector|VectorLengthSquared (Spline|GetArriveTangentatSplinePoint (Variables|Default|GetSpline) (Variables|GetSecondSplinePointIndex) "World")) (Math|Float|Square (ScalableFloat|GetValueatLevel (Variables|HF|GetCP_TangentThreshold)))))) (and (or (> (Math|Vector|VectorLengthSquared (Spline|GetTangentatSplinePoint (Variables|Default|GetSpline) (Variables|GetFirstSplinePointIndex) "World")) (Math|Float|Square (ScalableFloat|GetValueatLevel (Variables|HF|GetCP_TangentThreshold)))) (> (Math|Vector|VectorLengthSquared (Spline|GetArriveTangentatSplinePoint (Variables|Default|GetSpline) (Variables|GetSecondSplinePointIndex) "World")) (Math|Float|Square (ScalableFloat|GetValueatLevel (Variables|HF|GetCP_TangentThreshold))))) (< (Math|Vector|DotProduct (Math|Vector|Normalize (Spline|GetTangentatSplinePoint (Variables|Default|GetSpline) (Variables|GetFirstSplinePointIndex) "World")) (Math|Vector|Normalize (Spline|GetArriveTangentatSplinePoint (Variables|Default|GetSpline) (Variables|GetSecondSplinePointIndex) "World"))) (ScalableFloat|GetValueatLevel (Variables|HF|GetCP_FaceDirThreshold))))))
      (else
        (return false)))
    (else
      (return false))))

(fn SpawnGrindRailVariantComponent
  (bind _visualstyledata (Variables|VisualData|GetVisualStyleData))
  (bind _variantcomponenttype_35_b5a576f647a04e20f12e838713c958a2 (Utilities|Struct|BreakFGrindRailStyle _visualstyledata))
  (bind _grindrailvariantcomponent (Variables|Default|GetGrindRailVariantComponent))
  (if (not (Utilities|IsValidSoftClassReference (Variables|VisualData|GetLoadingVariantClass)))
    (Utilities|IsValid _grindrailvariantcomponent
      (:Is%20Valid
        (if (== (Utilities|ResolveSoftReference (Utilities|GetClass _grindrailvariantcomponent)) _variantcomponenttype_35_b5a576f647a04e20f12e838713c958a2)
          (Class|GrindRailVariantComponentBase|ApplyVariantVisuals _grindrailvariantcomponent (Variables|VisualData|GetVisualStyleData))
          (else
            (Class|GrindRailVariantComponentBase|CleanUpVisuals _grindrailvariantcomponent)
            (Components|DestroyComponent (Variables|Default|GetGrindRailVariantComponent)))))
      (:Is%20Not%20Valid
        _))
    (CallFunction|LoadandSpawnVariantComponent _variantcomponenttype_35_b5a576f647a04e20f12e838713c958a2)))

(fn SpawnTipComponent
  (bind _spawnedcomponent (Fortnite|Utilities|SpawnTransientSceneComponent (Variables|Default|GetGrindRail_TipComponent%20Class) self (Variables|Default|GetDefaultSceneRoot))) ; Transient to prevent sending the component across UEFN LiveEdit
  (bind _asgrind_rail_tip_component (Utilities|Casting|CastToGrindRail_TipComponent _spawnedcomponent)
    (:then
      (Class|GrindRailVariantComponentBase|UpdateTipMesh (Variables|Default|GetGrindRailVariantComponent) _asgrind_rail_tip_component))
    (:CastFailed
      _))
  (return _spawnedcomponent))

(fn UpdateVisibilityinGame
  (CallFunction|InitializeSceneComponentsforVisibility)
  (CallFunction|UpdateCollision))

(fn UpdateCollision
  (bind _creative_enabled_component (Variables|Default|GetCreative_Enabled_Component))
  (bind _enabled (Class|CreativeEnabledComponent|IsEnabled _creative_enabled_component))
  (bind _bvisible (CallFunction|IsVisible))
  (Utilities|FlowControl|Sequence
    (:then_0
      (if (Class|CreativeCommonDeviceFunctionLibrary|IsinPlayMode self)
        (switch Utilities|FlowControl|Switch|SwitchonEGrindRailDisableCollision (Variables|UserOptions|Advanced|GetDisableCollision)
          (:NewEnumerator0
            (LocalVariables|SetbEnableCollision true))
          (:NewEnumerator1
            (LocalVariables|SetbEnableCollision _enabled))
          (:NewEnumerator2
            (LocalVariables|SetbEnableCollision _bvisible))
          (:NewEnumerator3
            (LocalVariables|SetbEnableCollision (or _enabled _bvisible))))
        (else
          (LocalVariables|SetbEnableCollision true))))
    (:then_1
      (for _array_element (Variables|Default|GetSplineStaticMesh)
        (Collision|SetCollisionEnabled _array_element (select (LocalVariables|GetbEnableCollision) "QueryOnly" "NoCollision")))
      (if (not (LocalVariables|GetbEnableCollision))
        (CallFunction|DismountGrindingPlayers)))))

(fn UpdateMaterialsForVisibility
  (bind _grindrailvariantcomponent (Variables|Default|GetGrindRailVariantComponent))
  (bind _invisiblepreviewmaterial (Variables|VisualData|GetInvisiblePreviewMaterial))
  (Creative|Networking|Authority/Cosmetic
    (:Cosmetic
      (if (Default|ShouldUseHiddenMaterial)
        ;[1 1 1] In Edit Mode, apply a hologram material to the meshes instead of having the device be invisible
        (for _array_element (SplineProp|GetTransientSceneComponents :!class /Script/FortniteGame.SplinePropInterface)
          (Rendering|SetVisibility _array_element true false)
          (Development|SetHiddeninGame _array_element false false)
          (bind _asprimitive_component (Utilities|Casting|CastToPrimitiveComponent _array_element))
          (Rendering|Material|SetMaterial _asprimitive_component 0 _invisiblepreviewmaterial)
          (Rendering|Material|SetMaterial _asprimitive_component 1 _invisiblepreviewmaterial))
        (else
          (for _array_element_2 (Variables|Default|GetSplineStaticMesh)
            (for (_array_element_3 _array_index) (Default|GetEmissiveMaterials)
              (Rendering|Material|SetMaterial _array_element_2 _array_index _array_element_3)))
          (Utilities|IsValid _grindrailvariantcomponent
            (:Is%20Valid
              (Class|GrindRailVariantComponentBase|SetDefaultMaterials _grindrailvariantcomponent))))))))

(fn UpdateRailMeshColors
  (bind (_railcolor _coloremissivescale) (Default|GetRailColorAndEmissive))
  (bind (_returnvalue_r _returnvalue_g _returnvalue_b _returnvalue_a) (Math|Color|BreakColor (* _railcolor _coloremissivescale)))
  (for _array_element (Variables|Default|GetSplineStaticMesh)
    (Rendering|Material|SetDefaultCustomPrimitiveDataVector4 _array_element 2 (Math|Vector4|MakeVector4 _returnvalue_r _returnvalue_g _returnvalue_b _returnvalue_a))))

(fn UpdateSpeedSettings
  (LocalVariables|SetCachedBaseMaxSpeed (ScalableFloat|GetValueatLevelwithContext (LocalVariables|GetSF_BaseMaxSpeed)))
  (Utilities|TrySetwithContext (LocalVariables|GetSF_ShootingSpeedMultiplier) (Variables|UserOptions|Overrides|GetShootingSpeedMult_Override) (Variables|UserOptions|Basic|GetShootingSpeedMult))
  (Utilities|TrySetwithContext (LocalVariables|GetSF_SprintAcceleration) (Variables|UserOptions|Overrides|GetAcceleration_Sprint_Override) (Variables|UserOptions|Advanced|GetAcceleration_Sprint))
  (Utilities|TrySetwithContext (LocalVariables|GetSF_MinLeanAcceleration) (Variables|UserOptions|Overrides|GetAcceleration_Lean_Override) (Math|Float|SelectFloat (LocalVariables|GetDefaultMinLeanAcceleration) (Math|Float|Max(Float) (- (Variables|UserOptions|Advanced|GetAcceleration_Lean) (LocalVariables|GetDefaultMinLeanAcceleration))) (> (Variables|UserOptions|Advanced|GetAcceleration_Lean) (LocalVariables|GetDefaultMaxLeanAcceleration))))
  (Utilities|TrySetwithContext (LocalVariables|GetSF_MaxLeanAcceleration) (Variables|UserOptions|Overrides|GetAcceleration_Lean_Override) (Variables|UserOptions|Advanced|GetAcceleration_Lean))
  (Utilities|TrySetwithContext (LocalVariables|GetSF_GravityForceWhenGoingUp) (Variables|UserOptions|Overrides|GetGravityForce_GoingUp_Override) (Variables|UserOptions|Advanced|GetGravityForce_GoingUp))
  (Utilities|TrySetwithContext (LocalVariables|GetSF_GravityForceWhenGoingDown) (Variables|UserOptions|Overrides|GetGravityForce_GoingDown_Override) (Variables|UserOptions|Advanced|GetGravityForce_GoingDown))
  (Utilities|TrySetwithContext (LocalVariables|GetSF_BaseSpeed) (Variables|UserOptions|Overrides|GetBase%20Speed_Override) (Variables|UserOptions|Advanced|GetBase%20Speed))
  (Utilities|TrySetwithContext (LocalVariables|GetSF_BaseMaxSpeed) (Variables|UserOptions|Overrides|GetBase%20Max%20Speed_Override) (Variables|UserOptions|Advanced|GetBase%20Max%20Speed))
  (Utilities|TrySetwithContext (LocalVariables|GetSF_SpeedForMinLeanAcceleration) (Variables|UserOptions|Overrides|GetBase%20Max%20Speed_Override) (select (!= (ScalableFloat|GetValueatLevelwithContext (LocalVariables|GetSF_SpeedForMinLeanAcceleration)) 0) (* (/ (Variables|UserOptions|Advanced|GetBase%20Speed) (ScalableFloat|GetValueatLevelwithContext (LocalVariables|GetSF_SpeedForMinLeanAcceleration))) (LocalVariables|GetCachedBaseMaxSpeed)) (Math|Float|Max(Float) (Variables|UserOptions|Advanced|GetBase%20Speed) (LocalVariables|GetCachedBaseMaxSpeed))))
  (Utilities|TrySetwithContext (LocalVariables|GetSF_SprintMaxSpeed) (Variables|UserOptions|Overrides|GetSprinting%20Max%20Speed_Override) (Variables|UserOptions|Advanced|GetSprinting%20Max%20Speed))
  (Utilities|TrySetwithContext (LocalVariables|GetSF_SpeedHardCap) (Variables|UserOptions|Overrides|GetMax%20Speed%20Hard%20Cap_Override) (Variables|UserOptions|Advanced|GetMax%20Speed%20Hard%20Cap))
  (Utilities|TrySetwithContext (LocalVariables|GetSF_MinBaseAcceleration) (Variables|UserOptions|Overrides|GetBase%20Acceleration_Override) (Math|Float|SelectFloat (LocalVariables|GetDefaultMinBaseAcceleration) (Math|Float|Max(Float) (- (Variables|UserOptions|Advanced|GetBase%20Acceleration) (LocalVariables|GetDefaultMinBaseAcceleration)) 1e+01) (> (Variables|UserOptions|Advanced|GetBase%20Acceleration) (LocalVariables|GetDefaultMaxBaseAcceleration))))
  (Utilities|TrySetwithContext (LocalVariables|GetSF_MaxBaseAcceleration) (Variables|UserOptions|Overrides|GetBase%20Acceleration_Override) (Variables|UserOptions|Advanced|GetBase%20Acceleration))
  (Utilities|TrySetwithContext (LocalVariables|GetSF_SpeedForMinBaseAcceleration) (Variables|UserOptions|Overrides|GetBase%20Acceleration_Override) (select (!= (ScalableFloat|GetValueatLevelwithContext (LocalVariables|GetDefaultSpeedForMinBaseAcceleration)) 0) (* (/ (Variables|UserOptions|Advanced|GetBase%20Speed) (ScalableFloat|GetValueatLevelwithContext (LocalVariables|GetDefaultSpeedForMinBaseAcceleration))) (LocalVariables|GetCachedBaseMaxSpeed)) (Math|Float|Max(Float) (Variables|UserOptions|Advanced|GetBase%20Speed) (LocalVariables|GetCachedBaseMaxSpeed))))
  (Utilities|TrySetwithContext (LocalVariables|GetSF_MinStartSpeed) (Variables|UserOptions|Overrides|GetMinimum%20Start%20Speed_Override) (Variables|UserOptions|Advanced|GetMinimum%20Start%20Speed))
  (Utilities|TrySetwithContext (LocalVariables|GetSF_MaxStartSpeed) (Variables|UserOptions|Overrides|GetMaximum%20Start%20Speed_Override) (Variables|UserOptions|Advanced|GetMaximum%20Start%20Speed))
  (Utilities|TrySetwithContext (LocalVariables|GetSF_LandingSpeedBoost) (Variables|UserOptions|Overrides|GetLanding%20Speed%20Boost_Override) (Variables|UserOptions|Advanced|GetLanding%20Speed%20Boost))
  (Utilities|TrySetwithContext (LocalVariables|GetSF_JumpOffDistanceMultiplier) (Variables|UserOptions|Overrides|GetJump%20Off%20Distance%20Multiplier_Override) (Variables|UserOptions|Advanced|GetJump%20Off%20Distance%20Multiplier))
  (Utilities|TrySetwithContext (LocalVariables|GetSF_MinimumJumpForwardSpeed) (Variables|UserOptions|Overrides|GetMinimum%20Jump%20Forward%20Speed_Override) (Variables|UserOptions|Advanced|GetMinimum%20Jump%20Forward%20Speed))
  (Utilities|TrySetwithContext (LocalVariables|GetSF_JumpHeight) (Variables|UserOptions|Overrides|GetJump%20Height_Override) (Variables|UserOptions|Advanced|GetJump%20Height))
  (Utilities|TrySetwithContext (LocalVariables|GetSF_MaxDragForSpeedSoftCap) (Variables|UserOptions|Overrides|GetDragForSpeedSoftCap_Override) (Variables|UserOptions|Advanced|GetDragForSpeedSoftCap))
  (Utilities|TrySetwithContext (LocalVariables|GetSF_ShouldEnforceMinSpeed) true (Math|Conversions|ToFloat(Boolean) true))
  (Utilities|TrySetwithContext (LocalVariables|GetSF_MinimumSpeed) true (Variables|UserOptions|Advanced|GetMinimumSpeed))
  (Utilities|TrySetwithContext (LocalVariables|GetSF_ForcedDirection) true (Math|Conversions|ToFloat(Integer) (Utilities|Map|Find (Variables|Default|GetForcedDirectionMapping) (Variables|UserOptions|Advanced|GetForcedDirection))))
  (Utilities|TrySetwithContext (LocalVariables|GetSF_CanTurnAround) true (Math|Conversions|ToFloat(Boolean) (and (Variables|UserOptions|Advanced|GetbCanTurnAround) (Utilities|Enum|Equal(Enum) (Variables|UserOptions|Advanced|GetForcedDirection) "NewEnumerator0")))))

(fn UpdateSpineMeshAtIndex
  (params (SplineMeshIndex :type Integer))
  (bind _spline (Variables|Default|GetSpline))
  (bind _returnvalue (Spline|GetLocationatSplinePoint _spline (Variables|GetSplineMeshIndex) "Local"))
  (bind _splinemeshindex (Variables|GetSplineMeshIndex))
  (bind _returnvalue_2 (+ _splinemeshindex 1))
  (bind _spline_2 (Variables|Default|GetSpline))
  (bind _returnvalue_3 (Spline|GetTangentatSplinePoint _spline_2 _returnvalue_2 "Local"))
  (bind _spline_3 (Variables|Default|GetSpline))
  (bind _returnvalue_4 (Spline|GetLocationatSplinePoint _spline_3 _returnvalue_2 "Local"))
  (bind _spline_4 (Variables|Default|GetSpline))
  (bind _returnvalue_5 (Spline|GetTangentatSplinePoint _spline_4 _splinemeshindex "Local"))
  (bind _splinemeshindex_2 (Variables|GetSplineMeshIndex))
  (bind _returnvalue_6 (+ _splinemeshindex_2 1))
  (bind _staticmesh (Default|SelectSolidMeshforSegment _splinemeshindex_2 _returnvalue_6))
  (bind _splinemeshindex_3 (Variables|GetSplineMeshIndex))
  (bind _splinestaticmesh (Variables|Default|GetSplineStaticMesh))
  (bind _output (Utilities|Array|Get(acopy) _splinestaticmesh _splinemeshindex_3))
  (bind _spline_5 (Variables|Default|GetSpline))
  (bind _relativelocation (Variables|/Script/Engine.SceneComponent|GetRelativeLocation _spline_5))
  (bind _returnvalue_7 (+ _relativelocation _returnvalue_4))
  (bind _spline_6 (Variables|Default|GetSpline))
  (bind _relativelocation_2 (Variables|/Script/Engine.SceneComponent|GetRelativeLocation _spline_6))
  (bind _returnvalue_8 (+ _relativelocation_2 _returnvalue))
  (if (and (Utilities|Array|IsValidIndex (Variables|Default|GetSplineStaticMesh) SplineMeshIndex) (Math|Integer|InRange(Integer) SplineMeshIndex 0 (Spline|GetNumberOfSplinePoints (Variables|Default|GetSpline)) true false))
    (LocalVariables|SetSplineMshComp _output)
    (if (!= (Variables|/Script/Engine.StaticMeshComponent|GetStaticMesh _output) _staticmesh)
      ;[1 1 1] Update mesh based on the orientation of the spline points
      (Transformation|SetMobility (LocalVariables|GetSplineMshComp) "Movable") ; Has to be Movable to set a new mesh
      (Components|StaticMesh|SetStaticMesh (LocalVariables|GetSplineMshComp) _staticmesh)
      (Transformation|SetMobility (LocalVariables|GetSplineMshComp) "Static")
      (else
        _))
    (SplineMesh|SetStartandEnd (LocalVariables|GetSplineMshComp) _returnvalue_8 _returnvalue_5 _returnvalue_7 _returnvalue_3 true :!class /Script/Engine.SplineMeshComponent)
    (if (Utilities|Array|IsValidIndex (Variables|Default|GetInteractionMeshes) (Variables|GetSplineMeshIndex))
      (SplineMesh|SetStartandEnd (Utilities|Array|Get(acopy) (Variables|Default|GetInteractionMeshes) (Variables|GetSplineMeshIndex)) _returnvalue_8 _returnvalue_5 _returnvalue_7 _returnvalue_3 true :!class /Script/Engine.SplineMeshComponent))))

(fn UpdateTips
  (bind _returnvalue_location (Math|Transform|BreakTransform (Spline|GetTransformatDistanceAlongSpline (Variables|Default|GetSpline) 0.0 "World" false)))
  (bind _returnvalue_location_2 (Math|Transform|BreakTransform (Spline|GetTransformatDistanceAlongSpline (Variables|Default|GetSpline) (Spline|GetSplineLength (Variables|Default|GetSpline)) "World" false)))
  (if (and (Utilities|IsValid (Variables|Default|GetStartTip) :!class /Script/Engine.KismetSystemLibrary) (Utilities|IsValid (Variables|Default|GetEndTip) :!class /Script/Engine.KismetSystemLibrary))
    (Transformation|SetWorldLocationAndRotation (Variables|Default|GetStartTip) (+ _returnvalue_location (* (Transformation|GetUpVector (Variables|Default|GetStartTip)) (LocalVariables|GetTipVerticalOffset))) (Default|GetRotationForTip false) false false)
    (Transformation|SetWorldLocationAndRotation (Variables|Default|GetEndTip) (+ _returnvalue_location_2 (* (Transformation|GetUpVector (Variables|Default|GetEndTip)) (LocalVariables|GetTipVerticalOffset))) (Default|GetRotationForTip true) false false)
    (Class|GrindRailVariantComponentBase|UpdateTipMesh (Variables|Default|GetGrindRailVariantComponent) (Variables|Default|GetStartTip))
    (Class|GrindRailVariantComponentBase|UpdateTipMesh (Variables|Default|GetGrindRailVariantComponent) (Variables|Default|GetEndTip))
    (else
      (CallFunction|Tips))))

(fn UpdateVisuals
  (params (NumSplinePointsPreUpdate :type Integer))
  (bind _spline (Variables|Default|GetSpline))
  (bind _returnvalue (Spline|GetNumberOfSplinePoints _spline))
  (bind _returnvalue_2 (- _returnvalue 1))
  ;[1 1 1] Fixup existing spline meshes
  (for _index (range (+ (Math|Integer|Min(Integer) (Utilities|Array|Length (Variables|Default|GetSplineStaticMesh)) (- (Variables|GetNumSplinePointsPreUpdate) 2)) 1))
    (Spline|UpdateSpineMeshatIndex _index))
  ;[1 1 1] Add missing spline meshes as needed
  (for _index_2 (range (Math|Integer|Min(Integer) (Utilities|Array|Length (Variables|Default|GetSplineStaticMesh)) (- (Variables|GetNumSplinePointsPreUpdate) 1)) (+ (- (Spline|GetNumberOfSplinePoints (Variables|Default|GetSpline)) 2) 1))
    (CallFunction|CreateSplineMesh _index_2 (+ _index_2 1)))
  ;[1 1 1] Remove excess spline meshes as needed
  (if (>= (Utilities|Array|Length (Variables|Default|GetSplineStaticMesh)) (Spline|GetNumberOfSplinePoints (Variables|Default|GetSpline)))
    (for _index_3 (range (- (Spline|GetNumberOfSplinePoints (Variables|Default|GetSpline)) 1) (Utilities|Array|Length (Variables|Default|GetSplineStaticMesh)))
      (Spline|DestroySplineMeshatIndex _index_3))
    (Utilities|Array|Resize (Variables|Default|GetSplineStaticMesh) _returnvalue_2)
    (Utilities|Array|Resize (Variables|Default|GetInteractionMeshes) _returnvalue_2)
    (Utilities|Array|Resize (Variables|FortGrindRail|GetGrindRailMeshes) _returnvalue_2)
    (else
      _))
  (Spline|UpdateTips)
  (CallFunction|UpdateMaterialsforVisibility)
  (Default|UpdateInstancedStaticMeshes (Variables|Default|GetGrindRailVariantComponent)))

(fn UpdateVisualsForSpline
  (Spline|UpdateVisuals (Spline|GetNumberOfSplinePoints (Variables|Default|GetSpline))))

(fn ConstructionScript
  (CallFunction|InitializeSceneComponentsforVisibility))
