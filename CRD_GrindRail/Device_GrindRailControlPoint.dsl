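; Creative device for a grind rail control point, a movable node that shapes the rail's spline in edit mode.
; Property changes trigger a refresh, as do moves when no minigame is in progress.
; One tangent intensity option supplies both arrive and leave tangents, and caching stores only the leave value.
; Placement is valid only while absolute pitch stays under a threshold, and scale limits are fixed at one.
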(event OnPropertiesLoaded(ToyOptionsComponent)
  (CallFunction|OnPropertiesUpdated))

(event OnAnyPropertyChanged(ToyOptionsComponent)
  (CallFunction|OnPropertiesUpdated))

(event EventOnActorMovedToNewLocation
  ; Update the spline if device moved during edit mode
  (if (not (Minigame|IsMinigameinProgress (Variables|Default|GetFortMinigameProgress)))
    (CallFunction|OnPropertiesUpdated)))

(event EventCacheTangentIntensity
  (params (ArriveTangentIntensity :type Float%20%28single-precision%29) (LeaveTangentIntensity :type Float%20%28single-precision%29))
  (Variables|UserOptions|Basic|SetTangentIntensity LeaveTangentIntensity))

(fn GetArriveTangentIntensity
  (return (Variables|UserOptions|Basic|GetTangentIntensity)))

(fn GetCollisionStaticMesh
  (return "0, 0, 0" 0))

(fn GetLeaveTangentIntensity
  (return (Variables|UserOptions|Basic|GetTangentIntensity)))

(fn GetMeshComponents
  (return (Utilities|Array|MakeArray (Variables|Default|GetCreativeEditOnlyMesh) (Variables|Default|GetCreativeEditOnlyMesh_Arrow2) (Variables|Default|GetCreativeEditOnlyMesh_Arrow1))))

(fn GetScaleMax
  (params (DefaultScaleMax :type Float%20%28single-precision%29))
  (return 1.0))

(fn GetScaleMin
  (params (DefaultScaleMin :type Float%20%28single-precision%29))
  (return 1.0))

(fn HasAdditionalValidPlacementLogic
  (return true))

(fn HasValidPitch
  (return (< (Math|Float|Absolute(Float) (.pitch (.rotation (Transformation|GetActorTransform)))) (Variables|Default|GetPitchMaxThreshold))))

(fn IgnoreWorldGridScaleOverride
  (return false))

(fn IsValidPlacement
  (return (CallFunction|HasValidPitch)))

(fn ShouldBlockPlacement
  (params (MoveTool :type Fort%20Creative%20Move%20Tool%20Object%20Reference))
  (return (not (CallFunction|HasValidPitch))))

(fn ConstructionScript
  (|CallParentFunction))
