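; Creative Grind Vine device: a user-editable spline rail that players grind along, with configurable speed, visibility, collision and moss.
; The spline modification component feeds saved points into transient segment and tip meshes, avoiding LiveEdit replication and requiring transform fixup after saves.
; Enabled, visibility and minigame state drive collision and materials, including an edit-mode hologram when hidden.
; It tracks riders' grinding providers, broadcasts start/end events and force-dismounts them when collision turns off or rounds reset.
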(event EventEndPlay
  (params (EndPlayReason :type EEndPlayReason%20Enum :default "Destroyed"))
  (|CallParentFunction EndPlayReason)
  (Default|DismountGrindingPlayers))

(event Custom|OnTipComponentSpawned
  (bind _tiptype (Variables|UserOptions|Basic|GetTipType))
  (bind _tiptypetomesh (Variables|Default|GetTipTypeToMesh))
  (bind (_value _returnvalue) (Utilities|Map|Find _tiptypetomesh _tiptype))
  (bind _starttip (Variables|Default|GetStartTip))
  (bind _endtip (Variables|Default|GetEndTip))
  (if _returnvalue
    (bind _loaded_object (Creative|Loading|AsyncLoadEditorSafe _value)
      (:Loaded
        (bind _asstatic_mesh (Utilities|Casting|CastToStaticMesh _loaded_object))
        (Utilities|IsValid _starttip
          (:Is%20Valid
            (Components|StaticMesh|SetStaticMesh _starttip _asstatic_mesh))
          (:Is%20Not%20Valid
            _))
        (Utilities|IsValid _endtip
          (:Is%20Valid
            (Components|StaticMesh|SetStaticMesh _endtip _asstatic_mesh)))))))

(event OnAnyPropertyChanged(ToyOptionsComponent)
  (Class|CreativeEnabledComponent|Initialize (Variables|Default|GetCreative_Enabled_Component) (Variables|Default|GetFortMinigameProgress) (Variables|UserOptions|Advanced|GetEnabled%20During%20Phase))
  (Class|FortCreativeSplineModificationComponent|OnSavedSplineDataLoaded (Variables|Default|GetFortCreativeSplineModification) (Variables|SplinePoints|GetSplinePoints))
  (Default|UpdateVisibilityInGame)
  (Default|UpdateSpeedSettings)
  (if (not (Editor|InEditor))
    (Networking|SwitchHasAuthority
      (:Authority
        (Class|FortCreativeSplineModificationComponent|GenerateControlPointsforSpline (Variables|Default|GetFortCreativeSplineModification)))
      (:Remote
        (CallFunction|SpawnMoss)
        (CallFunction|OnTipComponentSpawned)))))

(event OnPropertiesLoaded(ToyOptionsComponent)
  (Class|CreativeEnabledComponent|Initialize (Variables|Default|GetCreative_Enabled_Component) (Variables|Default|GetFortMinigameProgress) (Variables|UserOptions|Advanced|GetEnabled%20During%20Phase))
  (Class|FortCreativeSplineModificationComponent|OnSavedSplineDataLoaded (Variables|Default|GetFortCreativeSplineModification) (Variables|SplinePoints|GetSplinePoints))
  (Default|UpdateVisibilityInGame)
  (Default|UpdateSpeedSettings)
  (if (not (Editor|InEditor))
    (Networking|SwitchHasAuthority
      (:Authority
        (Class|FortCreativeSplineModificationComponent|GenerateControlPointsforSpline (Variables|Default|GetFortCreativeSplineModification)))
      (:Remote
        (CallFunction|SpawnMoss)
        (CallFunction|OnTipComponentSpawned)))))

(event EditorPreview|EventUpdateEditorPreview
  (params (bHasMoved :type Boolean))
  ;[0.399 0.7428 1] Spline Editing
  (Class|FortCreativeSplineModificationComponent|UpdateEditorPreview (Variables|Default|GetFortCreativeSplineModification) (Variables|SplinePoints|GetSplinePoints)))

(event SplineProp|EventSaveSplinePointData
  ;[0.399 0.7428 1] Spline Editing
  ;;[0.4819 1 0.4834] SplinePropInterface Implementation
  (CallFunction|UpdatePropertyforLiveEdit (Utilities|String|ToString(Name) (Variables|/Script/FortniteGame.FortCreativeSplineModificationComponent|GetSplinePointsPropertyName (Variables|Default|GetFortCreativeSplineModification))))
  (CallFunction|Resave))

(event EditorPreview|EventSetupForEditorPreview
  ;[0.399 0.7428 1] Spline Editing
  (Default|UpdateMaterialsforVisibility)
  (if (Utilities|Array|IsEmpty (Variables|Default|GetSplineStaticMesh))
    (Default|UpdateVisuals (Spline|GetNumberOfSplinePoints (Variables|Default|GetSpline)))))

(event SplineProp|EventSetSerializedSplinePoints
  (params (SplinePoints :type Array%20of%20Spline%20Point%20Serializable%20Structures))
  ;[0.399 0.7428 1] Spline Editing
  ;;[0.4819 1 0.4834] SplinePropInterface Implementation
  ;;;[1 1 1 1 14] UEFN does not allow for serialized data on components, so allow the SplineModificationComponent to pass the data to the device to be saved
  (Variables|SplinePoints|SetSplinePoints SplinePoints))

(event SplineProp|EventOnSplinePointsChanged
  (params (NumSplinePointsPreChange :type Integer))
  ;[0.399 0.7428 1] Spline Editing
  ;;[0.4819 1 0.4834] SplinePropInterface Implementation
  (Default|UpdateVisuals NumSplinePointsPreChange)
  (Default|UpdateVisibilityInGame) ; Re-initialize the visible component with the new set of spline meshes
)

(event EventPostLevelSaveSpawnCleanup
  ;[0.399 0.7428 1] Spline Editing
  ;;[1 0.8866 0.4618] Since generated components are spawned as transient, the client/server is unable to resolve the references in the root component's AttachedChildren when the device is moved/rotated in UEFN, which leads to the AttachedChildren array being stomped with null elements and prevents the transform changes from propagating. So manually fix up the transforms here after the data has been serialized
  (|CallParentFunction)
  (Class|FortCreativeSplineModificationComponent|UpdateTransientComponentTransforms (Variables|Default|GetFortCreativeSplineModification)))

(event SplineProp|EventUpdateSplineMeshAtIndex
  (params (SplineMeshIndex :type Integer))
  ;[0.399 0.7428 1] Spline Editing
  ;;[0.4819 1 0.4834] SplinePropInterface Implementation
  (Default|UpdateSpineMeshatIndex SplineMeshIndex)
  (CallFunction|SpawnMoss))

(event EventOnPlayerBeganGrinding
  (params (GrindingPawn :type Fort%20Player%20Pawn%20Object%20Reference))
  (bind _returnvalue (Actor|GetComponentbyClass GrindingPawn "/GrindRail/Blueprints/BP_FortPawnGrindingProvider.BP_FortPawnGrindingProvider_C"))
  ;[1 0.5735 0.3708] Gameplay Events/Functions
  (Utilities|IsValid _returnvalue
    (:Is%20Valid
      (Utilities|Array|Add (Variables|Default|GetAttachedGrindingProviders) _returnvalue)
      (GameplayEvent|BroadcastEvent (Variables|UserOptions|Events|GetOn%20Started%20Grinding) 0 (Utilities|Struct|MakeGameplayEventGrindVinePlayerStartedGrinding (Pawn|GetController GrindingPawn) self)))))

(event EventOnPlayerEndedGrinding
  (params (Pawn :type Fort%20Player%20Pawn%20Object%20Reference))
  (bind _returnvalue (Actor|GetComponentbyClass Pawn "/GrindRail/Blueprints/BP_FortPawnGrindingProvider.BP_FortPawnGrindingProvider_C"))
  ;[1 0.5735 0.3708] Gameplay Events/Functions
  (Utilities|IsValid _returnvalue
    (:Is%20Valid
      (Utilities|Array|RemoveItem (Variables|Default|GetAttachedGrindingProviders) _returnvalue)
      (GameplayEvent|BroadcastEvent (Variables|UserOptions|Events|GetOn%20Ended%20Grinding) 0 (Utilities|Struct|MakeGameplayEventGrindVinePlayerEndedGrinding (Pawn|GetController Pawn) self)))))

(event Custom|EnableEvent
  ;[1 0.5735 0.3708] Gameplay Events/Functions
  (Class|CreativeEnabledComponent|SetEnabled (Variables|Default|GetCreative_Enabled_Component) true))

(event Custom|ShowTrackEvent
  ;[1 0.5735 0.3708] Gameplay Events/Functions
  (CallFunction|SetComponentsVisibility true))

(event Custom|DisableEvent
  ;[1 0.5735 0.3708] Gameplay Events/Functions
  (Class|CreativeEnabledComponent|SetEnabled (Variables|Default|GetCreative_Enabled_Component) false))

(event Custom|HideTrackEvent
  ;[1 0.5735 0.3708] Gameplay Events/Functions
  (CallFunction|SetComponentsVisibility false))

(event Custom|SetComponentsVisibility
  (params (bVisible :type Boolean))
  ;[1 0.5735 0.3708] Gameplay Events/Functions
  (for _array_element (Variables|Default|GetSplineStaticMesh)
    (Rendering|SetVisibility _array_element bVisible false)
    (Development|SetHiddeninGame _array_element (not bVisible) false))
  (Rendering|SetVisibility (Variables|Default|GetStartTip) bVisible false)
  (Development|SetHiddeninGame (Variables|Default|GetStartTip) (not bVisible) false)
  (Default|UpdateCollision)
  (Networking|CanExecuteCosmeticEvents
    (:True
      (CallFunction|SpawnMoss))))

(event MinigameStateChangedDelegate(FortMinigameProgress)
  (params (Minigame :type Fort%20Minigame%20Object%20Reference) (NewMinigameState :type EFortMinigameState%20Enum :default "PreGame"))
  ;[1 0.939 0.5021] Gameplay State Changes
  ;;[0.4745 1 0.8795] Stop grinding for players when the scoreboardstops showing after the game either in between rounds, starting a new match, or returning to edit mode
  (if (Math|Boolean|ORBoolean (Utilities|Enum|Equal(Enum) NewMinigameState "PostGameReset") (Utilities|Enum|Equal(Enum) NewMinigameState "Transitioning") (Utilities|Enum|Equal(Enum) NewMinigameState "PostGameAbandon"))
    (Class|CreativeEnabledComponent|SetEnabled (Variables|Default|GetCreative_Enabled_Component) false)
    (Default|DismountGrindingPlayers)))

(event OnEnabledStateChanged(Creative_Enabled_Component)
  (params (Enabled :type Boolean))
  ;[1 0.939 0.5021] Gameplay State Changes
  (Default|UpdateCollision)
  (if (and (not Enabled) (Variables|UserOptions|Advanced|GetbEndGrindingOnDisable))
    (Default|DismountGrindingPlayers)))

(event OnVisibilityChanged(Creative_VisibleInGame_Component)
  (params (Visible :type Boolean))
  ;[1 0.939 0.5021] Gameplay State Changes
  (Default|UpdateMaterialsforVisibility))

(event OnAddedtoMinigame(FortMinigameProgress)
  (params (Minigame :type Fort%20Minigame%20Object%20Reference))
  ;[1 0.939 0.5021] Gameplay State Changes
  (Utilities|IsValid Minigame
    (:Is%20Valid
      (Minigame|BindEventtoOnPlayModeChangedDelegate Minigame (delegate Custom|OnPlayModeChanged)))))

(event Custom|OnPlayModeChanged
  (params (Minigame :type Fort%20Minigame%20Object%20Reference) (bIsInPlayMode :type Boolean))
  ;[1 0.939 0.5021] Gameplay State Changes
  (Default|UpdateCollision))

(event Custom|LoadAndSetTipMeshMaterial
  (bind _tiptype (Variables|UserOptions|Basic|GetTipType))
  (bind _tiptypetomesh (Variables|Default|GetTipTypeToMesh))
  (bind (_value _returnvalue) (Utilities|Map|Find _tiptypetomesh _tiptype))
  (if _returnvalue
    (bind _loaded_object (Creative|Loading|AsyncLoadEditorSafe _value)
      (:Loaded
        (bind _asstatic_mesh (Utilities|Casting|CastToStaticMesh _loaded_object))
        (Rendering|Material|SetMaterial (Variables|Default|GetEndTip) 0 (StaticMesh|GetMaterial _asstatic_mesh 0))))))

(fn AddInteractionRailMeshComponent
  ;[1 1 1] AddSplineMeshComponent node spawns the component as Movable so that the static mesh can be set outside of this being called during construction
  (if (LocalVariables|GetbUseExpandedInteractionMesh)
    (bind _spawnedcomponent (Fortnite|Utilities|SpawnTransientSceneComponent "/CRD_GrindRail/CRD_GrindRailInteractionSplineMeshComponent.CRD_GrindRailInteractionSplineMeshComponent_C" self)) ; Transient to prevent sending the component across UEFN LiveEdit
    (Utilities|IsValid _spawnedcomponent
      (:Is%20Valid
        (Utilities|Array|Add (Variables|Default|GetInteractionMeshes) _spawnedcomponent)
        (Components|StaticMesh|SetStaticMesh _spawnedcomponent (Variables|Default|GetMesh_Interaction))
        (Transformation|SetMobility _spawnedcomponent "Static"))
      (:Is%20Not%20Valid
        _))
    (return _spawnedcomponent)
    (else
      (return 0))))

(fn AddSolidRailMeshComponent
  (params (SegmentLength :type Float%20%28double-precision%29) (FirstPoint :type Integer) (SecondPoint :type Integer))
  ;[1 1 1] Spawn the component as Movable so that the static mesh can be set outside of this being called during construction
  (bind _spawnedcomponent (Fortnite|Utilities|SpawnTransientSceneComponent "/CRD_GrindRail/CRD_GrindRailSplineMeshComponent.CRD_GrindRailSplineMeshComponent_C" self (Variables|Default|GetDefaultSceneRoot))) ; Transient to prevent sending the component across UEFN LiveEdit
  (Utilities|IsValid _spawnedcomponent
    (:Is%20Valid
      (bind _meshtouse (CallFunction|GetMeshforSplineLength SegmentLength))
      (Components|StaticMesh|SetStaticMesh _spawnedcomponent _meshtouse)
      (Transformation|SetMobility _spawnedcomponent "Static"))
    (:Is%20Not%20Valid
      _))
  (return _spawnedcomponent))

(fn AllowInverseScaling
  (return false))

(fn AllowSprinting
  (return (Variables|UserOptions|Basic|GetbAllowSprinting)))

(fn BlueprintCanInteract
  (params (InteractingPawn :type Fort%20Pawn%20Object%20Reference) (InteractionBeingAttempted :type EInteractionBeingAttempted%20Enum :default "FirstInteraction") (InteractionType :type TInteractionType%20Enum :default "IT_NoInteraction"))
  (return (or (CallFunction|IsGrindRailEnabled) (not (Minigame|IsinPlayMode self :!class /Script/FortniteGame.MinigameBlueprintLibrary)))))

(fn BlueprintGetInteractionString
  (params (InteractingPawn :type Fort%20Pawn%20Object%20Reference) (InteractionBeingAttempted :type EInteractionBeingAttempted%20Enum :default "FirstInteraction"))
  (if (Class|CreativeCommonDeviceFunctionLibrary|IsinPlayMode self)
    (bind _returnvalue (|CallParentFunction InteractingPawn InteractionBeingAttempted))
    (return _returnvalue)
    (else
      (return "NSLOCTEXT(\"[03F1BE0D7FF5897FB7828F4E318DE541]\", \"30CF545A43AB7262D22BE2895AE71DDF\", \"CUSTOMIZE GRIND RAIL\")"))))

(fn BlueprintOnLocalInteract
  (params (InteractingPawn :type Fort%20Player%20Pawn%20Object%20Reference) (InteractionBeingAttempted :type EInteractionBeingAttempted%20Enum :default "FirstInteraction"))
  (if (Class|CreativeCommonDeviceFunctionLibrary|IsinPlayMode self)
    (bind _returnvalue (|CallParentFunction InteractingPawn InteractionBeingAttempted))
    (return _returnvalue)
    (else
      (return false))))

(fn DestroySplineMeshAtIndex
  (params (MeshIndex :type Integer))
  (if (Utilities|Array|IsValidIndex (Variables|Default|GetSplineStaticMesh) MeshIndex)
    (Components|DestroyComponent (Utilities|Array|Get(acopy) (Variables|Default|GetSplineStaticMesh) (Variables|GetMeshIndex)))
    (else
      (Development|PrintString (Utilities|String|Append "[Grind Rail] Invalid Spline Mesh index: " (Utilities|String|ToString(Integer) MeshIndex)))))
  (if (Utilities|Array|IsValidIndex (Variables|Default|GetInteractionMeshes) (Variables|GetMeshIndex))
    (Components|DestroyComponent (Utilities|Array|Get(acopy) (Variables|Default|GetInteractionMeshes) (Variables|GetMeshIndex)))))

(fn DismountGrindingPlayers
  (bind _attachedgrindingproviders (Variables|Default|GetAttachedGrindingProviders))
  (bind _output_get (LocalVariables|SetTempAttachedGrindingProvider _attachedgrindingproviders))
  (for _array_element _output_get
    (Class|FortPawnComponentGrindingProvider|BPonDismount _array_element))
  (Utilities|Array|Clear _attachedgrindingproviders))

(fn ForceGridAlignment
  (return false))

(fn GetAdditionalSelectableActors
  (return 0))

(fn GetBlockPitchAndRoll
  (return false))

(fn GetCollisionStaticMesh
  (return "0, 0, 0" 0))

(fn GetMeshComponents
  (bind _starttip (Variables|Default|GetStartTip))
  (bind _endtip (Variables|Default|GetEndTip))
  (for _array_element (Variables|Default|GetSplineStaticMesh)
    (Utilities|Array|Add (LocalVariables|GetVisibleMeshes) _array_element))
  ;[1 0.9776 0.4567] Add rail tip meshes
  (Utilities|IsValid _starttip
    (:Is%20Valid
      (Utilities|Array|Add (LocalVariables|GetVisibleMeshes) _starttip))
    (:Is%20Not%20Valid
      _))
  (Utilities|IsValid _endtip
    (:Is%20Valid
      (Utilities|Array|Add (LocalVariables|GetVisibleMeshes) _endtip))
    (:Is%20Not%20Valid
      _))
  (return (LocalVariables|GetVisibleMeshes)))

(fn GetSavedSplinePoints
  (return (Variables|SplinePoints|GetSplinePoints)))

(fn GetScaleMax
  (params (DefaultScaleMax :type Float%20%28single-precision%29))
  (return 1.0))

(fn GetScaleMin
  (params (DefaultScaleMin :type Float%20%28single-precision%29))
  (return 1.0))

(fn GetSpawnedActors
  (return 0))

(fn GetSpeedSettingsOverride
  (return (Variables|Default|GetSpeedSettingsOverride) true))

(fn GetSplineComponent
  (return (Variables|Default|GetSpline)))

(fn GetTransientSceneComponents
  (bind _output_get (LocalVariables|SetComponents (Variables|Default|GetSplineStaticMesh)))
  (Utilities|Array|Add _output_get (Variables|Default|GetStartTip))
  (Utilities|Array|Add _output_get (Variables|Default|GetEndTip))
  (return (LocalVariables|GetComponents)))

(fn HasAdditionalSelectableActors
  (return false))

(fn HasAdditionalValidPlacementLogic
  (return true))

(fn HasValidRotation
  (return (and (< (Math|Float|Absolute(Float) (.roll (.rotation (Transformation|GetActorTransform)))) (Variables|Restrictions|GetRollMaxThreshold)) (< (Math|Float|Absolute(Float) (.pitch (.rotation (Transformation|GetActorTransform)))) (Variables|Restrictions|GetPitchMaxThreshold)))))

(fn IgnoreWorldGridScaleOverride
  (return false))

(fn IsGrindRailEnabled
  (return (Class|CreativeEnabledComponent|IsEnabled (Variables|Default|GetCreative_Enabled_Component))))

(fn IsValidPlacement
  (return (Default|HasValidRotation)))

(fn IsVisible
  (if (Utilities|Array|IsValidIndex (Variables|Default|GetSplineStaticMesh))
    (return (Rendering|IsVisible (Utilities|Array|Get(acopy) (Variables|Default|GetSplineStaticMesh))))
    (else
      (return false))))

(fn OnPreviewEnd
  (return false))

(fn OnPreviewStart
  (return false))

(fn SetSplineMeshMaterials
  (params (SplineMeshComp :type Spline%20Mesh%20Component%20Object%20Reference))
  (bind _invisiblepreviewmaterial (Variables|Default|GetInvisiblePreviewMaterial))
  (if (Utilities|ShouldUseHiddenMaterial)
    (Rendering|Material|SetMaterial SplineMeshComp 0 _invisiblepreviewmaterial)
    (Rendering|Material|SetMaterial SplineMeshComp 1 _invisiblepreviewmaterial)
    (else
      (Rendering|Material|SetMaterial SplineMeshComp 0 (Variables|Material|GetSplineVineMaterial))
      (Rendering|Material|SetMaterial SplineMeshComp 1 (Variables|Material|GetMaterialOfEmisssive)))))

(fn ShouldApplyFallDamageImmunity
  (return (Variables|UserOptions|Advanced|GetbApplyFallDamageImmunity)))

(fn ShouldBlockPlacement
  (params (MoveTool :type Fort%20Creative%20Move%20Tool%20Object%20Reference))
  (return (not (Default|HasValidRotation))))

(fn SpawnMoss
  (if (and (Variables|UserOptions|Basic|GetbCreateAdditionalGrass) (Visibility|ShouldbeVisible (Variables|Default|GetCreative_VisibleInGame_Component)))
    (|CallParentFunction)
    (else
      (Components|InstancedStaticMesh|ClearInstances (Variables|Default|GetMoss_HISM)))))

(fn SpawnTipComponent
  (bind _spawnedcomponent (Fortnite|Utilities|SpawnTransientSceneComponent (Variables|Default|GetGrindRail_TipComponent%20Class) self (Variables|Default|GetDefaultSceneRoot))) ; Transient to prevent sending the component across UEFN LiveEdit
  (CallFunction|OnTipComponentSpawned)
  (return _spawnedcomponent))

(fn UpdateVisibilityinGame
  (Initialization|InitializeWithSceneComponents (Variables|Default|GetCreative_VisibleInGame_Component) (SplineProp|GetTransientSceneComponents :!class /Script/FortniteGame.SplinePropInterface) (Utilities|Map|MakeMap) (Variables|UserOptions|Advanced|GetVisibleDuringPhase))
  (Default|UpdateCollision))

(fn UpdateCollision
  (bind _creative_enabled_component (Variables|Default|GetCreative_Enabled_Component))
  (bind _enabled (Class|CreativeEnabledComponent|IsEnabled _creative_enabled_component))
  (bind _bvisible (Default|IsVisible))
  (Utilities|FlowControl|Sequence
    (:then_0
      (if (Class|CreativeCommonDeviceFunctionLibrary|IsinPlayMode self)
        (switch Utilities|FlowControl|Switch|SwitchonEGrindRailDisableCollision (Variables|UserOptions|Advanced|GetDisableCollision)
          (:NewEnumerator0
            (LocalVariables|SetbEnableCollision true))
          (:NewEnumerator1
            (LocalVariables|SetbEnableCollision _enabled))
          (:NewEnumerator2
            (LocalVariables|SetbEnableCollision _bvisible))
          (:NewEnumerator3
            (LocalVariables|SetbEnableCollision (or _enabled _bvisible))))
        (else
          (LocalVariables|SetbEnableCollision true))))
    (:then_1
      (for _array_element (Variables|Default|GetSplineStaticMesh)
        (Collision|SetCollisionEnabled _array_element (select (LocalVariables|GetbEnableCollision) "QueryOnly" "NoCollision")))
      (if (not (LocalVariables|GetbEnableCollision))
        (Default|DismountGrindingPlayers)))))

(fn UpdateMaterialsForVisibility
  (bind _invisiblepreviewmaterial (Variables|Default|GetInvisiblePreviewMaterial))
  (bind _starttip (Variables|Default|GetStartTip))
  (Creative|Networking|Authority/Cosmetic
    (:Cosmetic
      (if (Utilities|ShouldUseHiddenMaterial)
        ;[1 1 1] In Edit Mode, apply a hologram material to the meshes instead of having the device be invisible
        (for _array_element (Variables|Default|GetSplineStaticMesh)
          (Rendering|SetVisibility _array_element true false)
          (Development|SetHiddeninGame _array_element false false)
          (Rendering|Material|SetMaterial _array_element 0 _invisiblepreviewmaterial)
          (Rendering|Material|SetMaterial _array_element 1 _invisiblepreviewmaterial))
        (Rendering|SetVisibility _starttip true false)
        (Development|SetHiddeninGame _starttip false false)
        (Rendering|Material|SetMaterial _starttip 0 (Variables|Default|GetInvisiblePreviewMaterial))
        (else
          (for _array_element_2 (Variables|Default|GetSplineStaticMesh)
            (Rendering|Material|SetMaterial _array_element_2 0 (Variables|Material|GetSplineVineMaterial))
            (Rendering|Material|SetMaterial _array_element_2 1 (Variables|Material|GetMaterialOfEmisssive)))
          (CallFunction|LoadandSetTipMeshMaterial) ; Load mesh from map to reset to default material (mesh should technically already be loaded)
        )))))

(fn UpdateSpeedSettings
  (Variables|Default|SetSpeedSettingsOverride (Utilities|Struct|MakeGrindRailSpeedSettings (Variables|UserOptions|Advanced|GetSprintAcceleration) (Variables|UserOptions|Advanced|GetLeanAcceleration) (Math|Float|SelectFloat (LocalVariables|GetDefaultMinLeanAcceleration) (Math|Float|Max(Float) (- (Variables|UserOptions|Advanced|GetLeanAcceleration) (LocalVariables|GetDefaultMinLeanAcceleration))) (> (Variables|UserOptions|Advanced|GetLeanAcceleration) (LocalVariables|GetDefaultMaxLeanAcceleration))) (Variables|UserOptions|Basic|GetShootingSpeedMultiplier) (Variables|UserOptions|Advanced|GetGravityForceWhenGoingUp) (Variables|UserOptions|Advanced|GetGravityForceWhenGoingDown))))

(fn UpdateSpineMeshAtIndex
  (params (SplineMeshIndex :type Integer))
  (bind _spline (Variables|Default|GetSpline))
  (bind _returnvalue (Spline|GetLocationatSplinePoint _spline (Variables|GetSplineMeshIndex) "Local"))
  (bind _splinemeshindex (Variables|GetSplineMeshIndex))
  (bind _returnvalue_2 (+ _splinemeshindex 1))
  (bind _spline_2 (Variables|Default|GetSpline))
  (bind _returnvalue_3 (Spline|GetTangentatSplinePoint _spline_2 _returnvalue_2 "Local"))
  (bind _spline_3 (Variables|Default|GetSpline))
  (bind _returnvalue_4 (Spline|GetLocationatSplinePoint _spline_3 _returnvalue_2 "Local"))
  (bind _spline_4 (Variables|Default|GetSpline))
  (bind _returnvalue_5 (Spline|GetTangentatSplinePoint _spline_4 _splinemeshindex "Local"))
  (bind _splinemeshindex_2 (Variables|GetSplineMeshIndex))
  (bind _splinestaticmesh (Variables|Default|GetSplineStaticMesh))
  (bind _output (Utilities|Array|Get(acopy) _splinestaticmesh _splinemeshindex_2))
  (bind _spline_5 (Variables|Default|GetSpline))
  (bind _relativelocation (Variables|/Script/Engine.SceneComponent|GetRelativeLocation _spline_5))
  (bind _returnvalue_6 (+ _relativelocation _returnvalue_4))
  (bind _spline_6 (Variables|Default|GetSpline))
  (bind _relativelocation_2 (Variables|/Script/Engine.SceneComponent|GetRelativeLocation _spline_6))
  (bind _returnvalue_7 (+ _relativelocation_2 _returnvalue))
  (if (and (Utilities|Array|IsValidIndex (Variables|Default|GetSplineStaticMesh) SplineMeshIndex) (Math|Integer|InRange(Integer) SplineMeshIndex 0 (Spline|GetNumberOfSplinePoints (Variables|Default|GetSpline)) true false))
    (LocalVariables|SetSplineMshComp _output)
    (bind _meshtouse (CallFunction|GetMeshforSplineLength (Math|Float|Absolute(Float) (- (Spline|GetDistanceAlongSplineatSplinePoint (Variables|Default|GetSpline) (Variables|GetSplineMeshIndex)) (Spline|GetDistanceAlongSplineatSplinePoint (Variables|Default|GetSpline) (+ (Variables|GetSplineMeshIndex) 1))))))
    (if (!= _meshtouse (Variables|/Script/Engine.StaticMeshComponent|GetStaticMesh _output))
      ;[1 1 1] Update mesh based on the orientation of the spline points
      (Transformation|SetMobility (LocalVariables|GetSplineMshComp) "Movable") ; Has to be Movable to set a new mesh
      (Components|StaticMesh|SetStaticMesh (LocalVariables|GetSplineMshComp) _meshtouse)
      (Transformation|SetMobility (LocalVariables|GetSplineMshComp) "Static")
      (else
        _))
    (SplineMesh|SetStartandEnd (LocalVariables|GetSplineMshComp) _returnvalue_7 _returnvalue_5 _returnvalue_6 _returnvalue_3 true :!class /Script/Engine.SplineMeshComponent)
    (if (Utilities|Array|IsValidIndex (Variables|Default|GetInteractionMeshes) (Variables|GetSplineMeshIndex))
      (SplineMesh|SetStartandEnd (Utilities|Array|Get(acopy) (Variables|Default|GetInteractionMeshes) (Variables|GetSplineMeshIndex)) _returnvalue_7 _returnvalue_5 _returnvalue_6 _returnvalue_3 true :!class /Script/Engine.SplineMeshComponent))))

(fn UpdateTips
  (bind (_returnvalue_location _returnvalue_rotation) (Math|Transform|BreakTransform (Spline|GetTransformatDistanceAlongSpline (Variables|Default|GetSpline) (- (Spline|GetSplineLength (Variables|Default|GetSpline)) (Variables|Default|GetSplineEndOffsetForTipRotation)) "World" false)))
  (bind _returnvalue_location_2 (Math|Transform|BreakTransform (Spline|GetTransformatDistanceAlongSpline (Variables|Default|GetSpline) (Spline|GetSplineLength (Variables|Default|GetSpline)) "World" false)))
  (bind _returnvalue_location_3 (Math|Transform|BreakTransform (Spline|GetTransformatDistanceAlongSpline (Variables|Default|GetSpline) 0.0 "World" false)))
  (bind (_returnvalue_location_4 _returnvalue_rotation_2) (Math|Transform|BreakTransform (Spline|GetTransformatDistanceAlongSpline (Variables|Default|GetSpline) (Variables|Default|GetSplineEndOffsetForTipRotation) "World" false)))
  (if (and (Utilities|IsValid (Variables|Default|GetStartTip) :!class /Script/Engine.KismetSystemLibrary) (Utilities|IsValid (Variables|Default|GetEndTip) :!class /Script/Engine.KismetSystemLibrary))
    (Transformation|SetWorldLocationAndRotation (Variables|Default|GetStartTip) (+ _returnvalue_location_3 (* (Transformation|GetUpVector (Variables|Default|GetStartTip)) (LocalVariables|GetTipVerticalOffset))) _returnvalue_rotation_2 false false)
    (Transformation|SetWorldLocationAndRotation (Variables|Default|GetEndTip) (+ _returnvalue_location_2 (* (Transformation|GetUpVector (Variables|Default|GetEndTip)) (LocalVariables|GetTipVerticalOffset))) _returnvalue_rotation false false)
    (else
      (CallFunction|Tips))))

(fn UpdateVisuals
  (params (NumSplinePointsPreUpdate :type Integer))
  (bind _spline (Variables|Default|GetSpline))
  (bind _returnvalue (Spline|GetNumberOfSplinePoints _spline))
  (bind _returnvalue_2 (- _returnvalue 1))
  ;[1 1 1] Fixup existing spline meshes
  (for _index (range (+ (Math|Integer|Min(Integer) (Utilities|Array|Length (Variables|Default|GetSplineStaticMesh)) (- (Variables|GetNumSplinePointsPreUpdate) 2)) 1))
    (Default|UpdateSpineMeshatIndex _index))
  ;[1 1 1] Add missing spline meshes as needed
  (for _index_2 (range (Math|Integer|Min(Integer) (Utilities|Array|Length (Variables|Default|GetSplineStaticMesh)) (- (Variables|GetNumSplinePointsPreUpdate) 1)) (+ (- (Spline|GetNumberOfSplinePoints (Variables|Default|GetSpline)) 2) 1))
    (CallFunction|CreateSplineMesh _index_2 (+ _index_2 1)))
  ;[1 1 1] Remove excess spline meshes as needed
  (if (>= (Utilities|Array|Length (Variables|Default|GetSplineStaticMesh)) (Spline|GetNumberOfSplinePoints (Variables|Default|GetSpline)))
    (for _index_3 (range (- (Spline|GetNumberOfSplinePoints (Variables|Default|GetSpline)) 1) (Utilities|Array|Length (Variables|Default|GetSplineStaticMesh)))
      (Default|DestroySplineMeshatIndex _index_3))
    (Utilities|Array|Resize (Variables|Default|GetSplineStaticMesh) _returnvalue_2)
    (Utilities|Array|Resize (Variables|Default|GetInteractionMeshes) _returnvalue_2)
    (Utilities|Array|Resize (Variables|FortGrindRail|GetGrindRailMeshes) _returnvalue_2)
    (else
      _))
  (Default|UpdateTips)
  (CallFunction|SpawnMoss))

(fn ConstructionScript
  (Initialization|InitializeWithSceneComponents (Variables|Default|GetCreative_VisibleInGame_Component) (SplineProp|GetTransientSceneComponents :!class /Script/FortniteGame.SplinePropInterface) (Utilities|Map|MakeMap) (Variables|UserOptions|Advanced|GetVisibleDuringPhase)) ; Init VisibleComponent in constructor to ensure that the proper material is applied to the spline meshes in UEFN
)
