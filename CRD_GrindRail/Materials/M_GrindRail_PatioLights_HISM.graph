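# Material for grind-rail patio string lights on instanced meshes, combining glowing, flickering bulbs with sway animation.
# Layered sine waves build flicker, offset by object position, while an emissive chain blends diffuse color with a warm tint using Fresnel, time of day and a spec-texture mask.
# Motion logic borrowed from foliage samples wind-actor data, an animation lookup texture and physics velocity fields, with rotation about an instance-space pivot declared as WPO.
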Material /CRD_GrindRail/Materials/M_GrindRail_PatioLights_HISM
graph {
 properties {
  bUsedWithInstancedStaticMeshes=true
  bUsedWithNanite=true
  bUsedWithSplineMeshes=true
  numCustomizedUVs=4
  pixelDepthOffsetMode=PDOM_Legacy
  usedShadingModels=Default Lit
 }
 nodes [
  n1 = NamedRerouteUsage()
  n2 = MaterialFunctionCall(Float3=@n1) {materialFunction=&BreakOutFloat3Components}
  n3 = Transform(None=Normalize(VectorInput=MaterialFunctionCall(X=@n2.G, Y=@n2.G, Z=@n2.G) {materialFunction=&MakeFloat3}.Result)) {transformSourceType=TRANSFORMSOURCE_Local}
  n4 = Power(Base=Multiply(A=Length(None=@n1) {is=class MaterialExpressionLength}) {constB=0.004})
  n5 = RotateAboutAxis(
   NormalizedRotationAxis=@n3
   RotationAngle=@n4
   PivotPoint=TransformPosition(None=Constant3Vector() {constant=<rgba 0 0 5 0>}) {transformSourceType=TRANSFORMPOSSOURCE_Instance, transformType=TRANSFORMPOSSOURCE_World}
   Position=WorldPosition().XYZ
  )
  n6 = Add(A=@n5, B=WorldPosition() {worldPositionShaderOffset=WPT_ExcludeAllShaderOffsets}.XYZ)
  n7 = TransformPosition(None=@n6) {transformSourceType=TRANSFORMPOSSOURCE_World, transformType=TRANSFORMPOSSOURCE_Instance}
  n8 = MaterialFunctionCall() {materialFunction=&ObjectPivotPoint}
  Add(A=@n5, B=Subtract(A=Add(A=@n7, B=@n8.Object Pivot Location), B=@n6))
  Add(A=@n5)
  Constant3Vector() {desc=Test Position, constant=<rgba 2 50 0.1 0>}
  Divide() {constB=200}
  n9 = Multiply(A=NamedRerouteUsage(), B=ScalarParameter() {defaultValue=1, group=AnimationTweaking, parameterName=Overall_MaxBend_0to1})
  n10 = NamedRerouteUsage()
  n11 = ScalarParameter() {defaultValue=1, group=Animation Physics, parameterName=Overall_PhysicsField_MaxBend}
  n12 = StaticSwitch(True=LinearInterpolate(A=@n9, B=Multiply(A=@n10, B=@n11), Alpha=NamedRerouteUsage()), False=@n9, Value=NamedRerouteUsage())
  LinearInterpolate(A=@n12, B=ScalarParameter() {defaultValue=1, group=Animation, parameterName=Overall_PlayerBend}, Alpha=NamedRerouteUsage())
  MaterialFunctionCall(Rotation Axis=@n3, Rotation Angle=@n4) {materialFunction=&FixRotateAboutAxisNormals}
  MaterialFunctionCall!(
   Previous WPO Value
   Force Magnitude Mask
   Force Magnitude Multiplier
   Project Field Force Vector to Normals
   Noise Size
   Noise Pan Speed
   Noise Falloff
   Stretch Noise Coordinates Z
  ) {desc=Sample global physics field add to WPO, materialFunction=/Game/Effects/Niagara/MasterMaterials/Functions/MF_FX_PhysicsFields_Grass}
  n13 = Multiply(A=Time(), B=ScalarParameter() {defaultValue=1, group=Emissive, parameterName=Flicker Speed})
  n14 = Add(A=Multiply(A=Sine(None=@n13) {period=3}) {constB=0.5}) {constB=0.5}
  n15 = Add(A=@n13, B=NamedRerouteUsage())
  n16 = Multiply(A=Sine(None=@n15) {period=0.5}) {constB=0.1}
  n17 = Sine(None=@n15) {period=0.36}
  n18 = Sine(None=@n15) {period=0.82}
  n19 = Add(A=Multiply(A=Sine(None=@n15)) {constB=0.1}, B=Add(A=Multiply(A=@n17) {constB=0.1}, B=Multiply(A=@n18) {constB=0.1}))
  n20 = LinearInterpolate(
   A=ScalarParameter() {defaultValue=0.5, parameterName=Flicker Min}
   B=ScalarParameter() {defaultValue=1, parameterName=Flicker Max}
   Alpha=Add(A=LinearInterpolate(Alpha=@n14) {constA=0.5, constB=1.2}, B=Add(A=@n16, B=@n19))
  )
  MaterialFunctionCall(Number=@n20) {desc=Debug Offset, materialFunction=&DebugScalarValues}
  n21 = ObjectPositionWS()
  n22 = ScalarParameter() {defaultValue=256, parameterName=Randomness Scale}
  n23 = Multiply(A=Divide(A=ComponentMask(None=@n21) {r=true}, B=@n22)) {constB=0.71}
  n24 = Divide(A=ComponentMask(None=@n21) {g=true}, B=@n22)
  n25 = Divide(A=ComponentMask(None=@n21) {b=true}, B=@n22)
  n26 = Subtract(A=Add(A=Frac(None=@n23), B=Sine(None=@n24)), B=Frac(None=Multiply(A=@n25) {constB=0.37}))
  MaterialFunctionCall(Number=@n26) {desc=Debug Offset, materialFunction=&DebugScalarValues}
  n27 = Multiply(A=NamedRerouteUsage(), B=Constant() {desc=Min Brightness, r=0.75})
  n28 = Constant3Vector() {constant=<rgba 1 0.3945 0 1>}
  n29 = TextureSampleParameter2D() {parameterName=Diffuse, texture=/Game/Environments/Sets/BlackMonday/Textures/T_BM_StringLights_01_D}
  n30 = Constant3Vector() {constant=<rgba 80 39.7824 13.5786 1>}
  n31 = Constant() {r=0.5}
  n32 = MaterialFunctionCall(Value_Morning=@n31, Value_Day=Constant() {r=1}, Value_Evening=@n31, Value_Night=@n31) {materialFunction=&Scalar_Time_of_Day}
  n33 = Fresnel(ExponentIn=Constant() {r=0.05}) {exponent=0.25}
  n34 = Power(Base=Multiply(A=OneMinus(None=@n33)) {constB=5}) {constExponent=1.5}
  n35 = Multiply(A=Multiply(A=Multiply(A=@n30, B=@n32.Result), B=NamedRerouteUsage()), B=@n34)
  Multiply(
   A=LinearInterpolate(A=Multiply(A=@n27, B=@n28), B=Multiply(A=@n29.RGB, B=@n35), Alpha=@n34) {constA=0.25}
   B=ComponentMask(None=TextureSampleParameter2D() {parameterName=Spec, samplerType=SAMPLERTYPE_Masks, texture=/Game/Environments/Sets/BlackMonday/Textures/T_BM_StringLights_01_S}.A) {g=true}
  )
  NamedRerouteDeclaration(None=@n5) {name=WPO, nodeColor=<rgba 1 0.2578 0 1>}
  NamedRerouteDeclaration(None=@n20) {name=Flicker, nodeColor=<rgba 0.8165 1 0 1>}
  n36 = Constant3Vector()
  n37 = Multiply(
   A=NamedRerouteDeclaration(None=TransformPosition(None=@n36) {transformSourceType=TRANSFORMPOSSOURCE_Instance, transformType=TRANSFORMPOSSOURCE_World}) {name=BasePivot, nodeColor=<rgba 1 0 0.7031 1>}
   B=ScalarParameter() {defaultValue=1408.6052, group=AnimationTweaking, parameterName=Overall_Bend_Frequency}
  )
  n38 = StaticBool() {value=true}
  n39 = MaterialFunctionCall!(
   Position=@n37
   IsTrees
   IsGrass
   IsMediumPlant=@n38
   IsMediumPlant_2024Anim
   IsMediumPlant2ndLevel=StaticBool() {value=true}
   MediumPlant_BasePivotPostion=NamedRerouteUsage()
  ) {materialFunction=/Game/Environments/Asteria/Foliage/Wind/MF_WindActorData}
  NamedRerouteDeclaration(None=@n39.WindI_Direction_WithNoise) {name=WindDirectionOverallBend, nodeColor=<rgba 0 0.1562 1 1>}
  NamedRerouteDeclaration(None=@n39.WindIntesnity) {name=WindIntensityNoise_Overall, nodeColor=<rgba 1 0 0.4219 1>}
  n40 = Multiply(
   A=ComponentMask(None=TextureCoordinate() {coordinateIndex=1}) {g=true}
   B=ScalarParameter() {defaultValue=1308.7753, group=AnimationTweaking, parameterName=Leaf_RandomFrequency}
  )
  n41 = ScalarParameter() {defaultValue=246.1449, group=AnimationTweaking, parameterName=Leaf_DragWiggle_Calm}
  n42 = ScalarParameter() {defaultValue=120, group=AnimationTweaking, parameterName=Leaf_DragWiggle_Storm}
  n43 = NamedRerouteUsage()
  n44 = Add(A=@n40, B=Multiply(A=VertexColor().A, B=LinearInterpolate(A=@n41, B=@n42, Alpha=@n43)))
  n45 = AppendVector(A=@n44, B=@n44)
  n46 = MaterialFunctionCall(Position=Add(A=AppendVector(A=@n45, B=@n40), B=@n37), IsMediumPlant=@n38) {materialFunction=&MF_WindActorData}
  NamedRerouteDeclaration(None=@n46.StormMask) {name=StormMask, nodeColor=<rgba 0.3125 0 1 1>}
  n47 = VertexColor()
  n48 = TextureSample(UVs=AppendVector(A=@n12, B=OneMinus(None=@n47.A))) {
   samplerSource=SSM_Clamp_WorldGroupSettings
   samplerType=SAMPLERTYPE_LinearColor
   texture=/Game/Environments/Asteria/Foliage/Grass/MasterMaterial/GenericShared_Grass_Foliage_Anim
  }
  n49 = Multiply(A=@n48.R, B=ScalarParameter() {defaultValue=1.3, group=AnimationTweaking, parameterName=Overall_Bend_XY})
  n50 = NamedRerouteUsage()
  n51 = StaticSwitch(True=LinearInterpolate(A=@n50, B=NamedRerouteUsage(), Alpha=NamedRerouteUsage()), False=@n50, Value=NamedRerouteUsage())
  n52 = Constant2Vector()
  n53 = Multiply(A=Multiply(A=@n48.G) {constB=-1}, B=ScalarParameter() {defaultValue=1.3, group=AnimationTweaking, parameterName=Overall_Bend_Z})
  NamedRerouteDeclaration(None=Add(A=Multiply(A=@n49, B=@n51), B=AppendVector(A=@n52, B=@n53))) {name=Cheap OverallBend, nodeColor=<rgba 0.5 1 0 1>}
  NamedRerouteDeclaration(None=StaticBool() {value=true}) {name=Use Physics, nodeColor=<rgba 0 1 0.3438 1>}
  n54 = SamplePhysicsVectorField!(World Position=NamedRerouteUsage()) {fieldTarget=Vector_LinearVelocity}
  n55 = MaterialFunctionCall(Vector 3=Add(A=@n54) {constB=0.0001}) {materialFunction=/Engine/Functions/Engine_MaterialFunctions02/Utility/VectorLength}
  n56 = Constant()
  NamedRerouteDeclaration(None=ShadingPathSwitch(Default=Saturate(None=@n55.V3 Length), Mobile=@n56) {desc=Migrated from FeatureLevelSwitch}) {name=PhysicsFieldMask, nodeColor=<rgba 0 1 0.3906 1>}
  n57 = Constant3Vector() {constant=<rgba 0 0 1 0>}
  NamedRerouteDeclaration(None=ShadingPathSwitch(
    Default=MaterialFunctionCall(Vector=@n54, Default=@n57) {materialFunction=&SafeNormalize}.Result
    Mobile=Constant3Vector() {constant=<rgba 0 0 1 0>}
   ) {desc=Migrated from FeatureLevelSwitch}) {name=PhysicsFieldDirection, nodeColor=<rgba 0.3828 0 1 1>}
  n58 = Min(A=@n55.V3 Length, B=ScalarParameter() {defaultValue=1, group=PhysicsField, parameterName=MaxRotation_PhysicsField_Clamp})
  n59 = Time()
  n60 = VertexColor()
  n61 = Add(A=@n60.R, B=@n60.R)
  n62 = Multiply(
   A=ScalarParameter() {defaultValue=351, group=PhysicsField, parameterName=PhysicsField_NoiseOffset}
   B=Multiply(A=Add(A=@n61, B=@n60.B)) {constB=33.33}
  )
  n63 = Frac(None=Multiply(A=ScalarParameter() {defaultValue=2.5, group=PhysicsField, parameterName=MaxRotation_PhysicsField_NoiseSpeed}, B=Add(A=@n59, B=@n62)))
  n64 = OneMinus(None=@n63)
  n65 = LinearInterpolate(
   A=Multiply(A=Multiply(A=@n63, B=@n64)) {constB=2}
   Alpha=ScalarParameter() {defaultValue=0.6, group=PhysicsField, parameterName=MaxRotation_PhysicsField_MInBend}
  )
  NamedRerouteDeclaration(None=ShadingPathSwitch(Default=Multiply(A=@n58, B=@n65), Mobile=@n56) {desc=Migrated from FeatureLevelSwitch}) {name=PhysicsFieldStrength, nodeColor=<rgba 1 0 0.2109 1>}
  NamedRerouteDeclaration(None=@n26) {name=Time Random Offset, nodeColor=<rgba 1 1 0 1>}
  NamedRerouteDeclaration(None=@n46.WindIntensitySingleSample) {name=WindIntensityNoise, nodeColor=<rgba 0.1719 0 1 1>}
  NamedRerouteDeclaration(None=@n46.WindI_Direction_WithNoise) {name=WindDirectionWithNoise, nodeColor=<rgba 0 1 0.9297 1>}
  NamedRerouteUsage()
  NamedRerouteUsage()
  n66 = TextureCoordinate() {coordinateIndex=2}
  n67 = ComponentMask(None=TextureCoordinate() {coordinateIndex=3}) {r=true}
  n68 = TextureSampleParameter2D() {
   group=Texture
   parameterName=Normal
   samplerType=SAMPLERTYPE_Normal
   texture=/Game/Environments/Sets/BlackMonday/Textures/T_BM_StringLights_01_N
  }
  Reroute(None=MaterialFunctionCall(BaseNormal=AppendVector(A=@n66, B=@n67), AdditionalNormal=@n68.RGB) {materialFunction=&BlendAngleCorrectedNormals}.Result)
  Reroute(None=@n46.StormMask)
 ]
}
